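# Control Rig for the FruitFire Nudge female medium body that applies corrective motion to its dress and necklace bones and runs a shared deform-rig function.
# Spherical pose readers on the thighs and right clavicle fill alpha variables.
# Parent constraints bind dress bones to thighs and calves, with center bones averaging left and right.
# Products of alphas and state variables (crouch, emote, skydive, slope), mostly smoothed by AlphaInterp, weight additive ModifyTransforms correctives.
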ControlRigBlueprint /BRCosmetics/Characters/Player/Female/Medium/Bodies/F_MED_FruitFire_Nudge/Meshes/F_MED_FruitFire_Nudge_CtrlRig
graphs [
 RigVMFunctionLibrary {links [], nodes []}
 RigVMFunction_Sequence_1_ContainedGraph {nodes [
   n1 = RigVMFunctionEntryNode()
   RigVMFunction_Sequence = &RigVMFunction_Sequence(ExecuteContext=@n1.ExecuteContext)
   n2 = &RigVMFunction_Sequence(ExecuteContext=@RigVMFunction_Sequence.B)
   n3 = &RigVMFunction_Sequence(ExecuteContext=@n2.B)
   n4 = &RigVMFunction_Sequence(ExecuteContext=@n3.B)
   n5 = &RigVMFunction_Sequence(ExecuteContext=@n4.B)
   n6 = &RigVMFunction_Sequence(ExecuteContext=@n5.B)
   RigVMFunctionReturnNode(A=@RigVMFunction_Sequence.A, B=@n2.A, C=@n3.A, D=@n4.A, E=@n5.A, F=@n6.A, G=@n6.B)
  ]}
 RigVMModel {nodes [
   RigUnit_BeginExecution = &RigUnit_BeginExecution()
   RigVMFunction_Sequence = RigVMAggregateNode(ExecuteContext=@RigUnit_BeginExecution.ExecutePin)
   n1 = &RigUnit_SphericalPoseReader(
    ActiveRegionScaleFactors=(PositiveWidth=4,NegativeWidth=3,PositiveHeight=2.5,NegativeHeight=2)
    ActiveRegionSize=0.1
    Debug=(bDrawDebug=false,bDraw2D=False,bDrawLocalAxes=False,DebugScale=25,DebugSegments=20,DebugThickness=0.25)
    DriverAxis=(X=-1,Y=0,Z=0)
    DriverItem=(Type=Bone,Name="thigh_l")
    FalloffRegionScaleFactors=(PositiveWidth=1,NegativeWidth=3,PositiveHeight=1,NegativeHeight=0.5)
    FalloffSize=0.35
    FlipHeightScaling=False
    FlipWidthScaling=False
    OptionalParentItem=(Type=Bone,Name="None")
    RotationOffset=(X=100,Y=0,Z=-20)
    ExecutePin=@RigVMFunction_Sequence.A
   )
   n2 = RigVMVariableNode(Value=@n1.OutputParam, ExecuteContext=@RigVMFunction_Sequence.B) {variable=L thighForward Alpha}
   n3 = &RigUnit_SphericalPoseReader(
    ActiveRegionScaleFactors=(PositiveWidth=4,NegativeWidth=3,PositiveHeight=2.5,NegativeHeight=2)
    ActiveRegionSize=0.1
    Debug=(bDrawDebug=false,bDraw2D=False,bDrawLocalAxes=False,DebugScale=25,DebugSegments=20,DebugThickness=0.25)
    DriverAxis=(X=1,Y=0,Z=0)
    DriverItem=(Type=Bone,Name="thigh_r")
    FalloffRegionScaleFactors=(PositiveWidth=1,NegativeWidth=3.6,PositiveHeight=1,NegativeHeight=0.5)
    FalloffSize=0.4
    FlipHeightScaling=False
    FlipWidthScaling=true
    OptionalParentItem=(Type=Bone,Name="None")
    RotationOffset=(X=-88.1205,Y=0,Z=-14.0228)
    ExecutePin=@n1.ExecutePin
   )
   n4 = RigVMVariableNode(Value=@n3.OutputParam, ExecuteContext=@n2.ExecuteContext) {variable=R thighForward Alpha}
   n5 = &RigUnit_SphericalPoseReader(
    ActiveRegionScaleFactors=(PositiveWidth=3,NegativeWidth=2,PositiveHeight=2,NegativeHeight=2)
    ActiveRegionSize=0.1
    Debug=(bDrawDebug=false,bDraw2D=False,bDrawLocalAxes=False,DebugScale=25,DebugSegments=20,DebugThickness=0.25)
    DriverAxis=(X=-1,Y=0,Z=0)
    DriverItem=(Type=Bone,Name="thigh_l")
    FalloffRegionScaleFactors=(PositiveWidth=2,NegativeWidth=2,PositiveHeight=2,NegativeHeight=2)
    FalloffSize=0.25
    FlipHeightScaling=false
    FlipWidthScaling=false
    OptionalParentItem=(Type=Bone,Name="None")
    RotationOffset=(X=180,Y=-40,Z=0)
    ExecutePin=@n3.ExecutePin
   )
   n6 = RigVMVariableNode(Value=@n5.OutputParam, ExecuteContext=@n4.ExecuteContext) {variable=L thighSide Alpha}
   n7 = &RigUnit_SphericalPoseReader(
    ActiveRegionScaleFactors=(PositiveWidth=3,NegativeWidth=2,PositiveHeight=2,NegativeHeight=2)
    ActiveRegionSize=0.1
    Debug=(bDrawDebug=false,bDraw2D=False,bDrawLocalAxes=False,DebugScale=25,DebugSegments=20,DebugThickness=0.25)
    DriverAxis=(X=1,Y=0,Z=0)
    DriverItem=(Type=Bone,Name="thigh_r")
    FalloffRegionScaleFactors=(PositiveWidth=2,NegativeWidth=2,PositiveHeight=2,NegativeHeight=2)
    FalloffSize=0.25
    FlipHeightScaling=False
    FlipWidthScaling=true
    OptionalParentItem=(Type=Bone,Name="None")
    RotationOffset=(X=0,Y=-40,Z=0)
    ExecutePin=@n5.ExecutePin
   )
   n8 = RigVMVariableNode(Value=@n7.OutputParam, ExecuteContext=@n6.ExecuteContext) {variable=R thighSide Alpha}
   n9 = &RigUnit_SphericalPoseReader(
    ActiveRegionScaleFactors=(PositiveWidth=2,NegativeWidth=2,PositiveHeight=2,NegativeHeight=1)
    ActiveRegionSize=0.1
    Debug=(bDrawDebug=false,bDraw2D=False,bDrawLocalAxes=False,DebugScale=25,DebugSegments=20,DebugThickness=0.25)
    DriverAxis=(X=-1,Y=0,Z=0)
    DriverItem=(Type=Bone,Name="thigh_l")
    FalloffRegionScaleFactors=(PositiveWidth=2,NegativeWidth=2,PositiveHeight=4,NegativeHeight=1)
    FalloffSize=0.27
    FlipHeightScaling=false
    FlipWidthScaling=false
    OptionalParentItem=(Type=Bone,Name="None")
    RotationOffset=(X=0,Y=90,Z=46.3895)
    ExecutePin=@n7.ExecutePin
   )
   n10 = RigVMVariableNode(Value=@n9.OutputParam, ExecuteContext=@n8.ExecuteContext) {variable=L thighBack Alpha}
   n11 = &RigUnit_SphericalPoseReader(
    ActiveRegionScaleFactors=(PositiveWidth=2,NegativeWidth=2,PositiveHeight=2,NegativeHeight=1)
    ActiveRegionSize=0.1
    Debug=(bDrawDebug=false,bDraw2D=False,bDrawLocalAxes=False,DebugScale=25,DebugSegments=20,DebugThickness=0.25)
    DriverAxis=(X=1,Y=0,Z=0)
    DriverItem=(Type=Bone,Name="thigh_r")
    FalloffRegionScaleFactors=(PositiveWidth=2,NegativeWidth=2,PositiveHeight=3,NegativeHeight=1)
    FalloffSize=0.27
    FlipHeightScaling=false
    FlipWidthScaling=false
    OptionalParentItem=(Type=Bone,Name="None")
    RotationOffset=(X=211.7749,Y=84.6752,Z=60.6503)
    ExecutePin=@n9.ExecutePin
   )
   n12 = RigVMVariableNode(Value=@n11.OutputParam, ExecuteContext=@n10.ExecuteContext) {variable=R thighBack Alpha}
   n13 = &RigUnit_SphericalPoseReader(
    ActiveRegionScaleFactors=(PositiveWidth=1.5,NegativeWidth=1.5,PositiveHeight=1.5,NegativeHeight=1.5)
    ActiveRegionSize=0.1
    Debug=(bDrawDebug=false,bDraw2D=False,bDrawLocalAxes=False,DebugScale=25,DebugSegments=20,DebugThickness=0.25)
    DriverAxis=(X=-1,Y=0,Z=0)
    DriverItem=(Type=Bone,Name="thigh_l")
    FalloffRegionScaleFactors=(PositiveWidth=2,NegativeWidth=2,PositiveHeight=2,NegativeHeight=2)
    FalloffSize=0.23
    FlipHeightScaling=false
    FlipWidthScaling=false
    OptionalParentItem=(Type=Bone,Name="None")
    RotationOffset=(X=33.1118,Y=48.7758,Z=-2.2671)
    ExecutePin=@n11.ExecutePin
   )
   n14 = RigVMVariableNode(Value=@n13.OutputParam, ExecuteContext=@n12.ExecuteContext) {variable=L thighSideIn_Fwd Alpha}
   n15 = &RigUnit_SphericalPoseReader(
    ActiveRegionScaleFactors=(PositiveWidth=1.5,NegativeWidth=1.5,PositiveHeight=1.5,NegativeHeight=1.5)
    ActiveRegionSize=0.1
    Debug=(bDrawDebug=false,bDraw2D=False,bDrawLocalAxes=False,DebugScale=25,DebugSegments=20,DebugThickness=0.25)
    DriverAxis=(X=1,Y=0,Z=0)
    DriverItem=(Type=Bone,Name="thigh_r")
    FalloffRegionScaleFactors=(PositiveWidth=2,NegativeWidth=2,PositiveHeight=2,NegativeHeight=2)
    FalloffSize=0.23
    FlipHeightScaling=False
    FlipWidthScaling=true
    OptionalParentItem=(Type=Bone,Name="None")
    RotationOffset=(X=-166.5537,Y=44.1032,Z=13.3453)
    ExecutePin=@n13.ExecutePin
   )
   n16 = RigVMVariableNode(Value=@n15.OutputParam, ExecuteContext=@n14.ExecuteContext) {variable=R thighSideIn_Bwd Alpha}
   n17 = &RigUnit_SphericalPoseReader(
    ActiveRegionScaleFactors=(PositiveWidth=1,NegativeWidth=1,PositiveHeight=1,NegativeHeight=1)
    ActiveRegionSize=0.12
    Debug=(bDrawDebug=false,bDraw2D=False,bDrawLocalAxes=False,DebugScale=25,DebugSegments=20,DebugThickness=0.25)
    DriverAxis=(X=0,Y=1,Z=0)
    DriverItem=(Type=Bone,Name="thigh_l")
    FalloffRegionScaleFactors=(PositiveWidth=1,NegativeWidth=1,PositiveHeight=1,NegativeHeight=1)
    FalloffSize=0.17
    FlipHeightScaling=false
    FlipWidthScaling=false
    OptionalParentItem=(Type=Bone,Name="None")
    RotationOffset=(X=58.5993,Y=-9.6785,Z=-26.2186)
    ExecutePin=@n15.ExecutePin
   )
   n18 = RigVMVariableNode(Value=@n17.OutputParam, ExecuteContext=@n16.ExecuteContext) {variable=L thighTwistIn Alpha}
   n19 = &RigUnit_SphericalPoseReader(
    ActiveRegionScaleFactors=(PositiveWidth=1,NegativeWidth=1,PositiveHeight=1,NegativeHeight=1)
    ActiveRegionSize=0.12
    Debug=(bDrawDebug=false,bDraw2D=False,bDrawLocalAxes=False,DebugScale=25,DebugSegments=20,DebugThickness=0.25)
    DriverAxis=(X=0,Y=-1,Z=0)
    DriverItem=(Type=Bone,Name="thigh_r")
    FalloffRegionScaleFactors=(PositiveWidth=1,NegativeWidth=1,PositiveHeight=1,NegativeHeight=1)
    FalloffSize=0.17
    FlipHeightScaling=False
    FlipWidthScaling=true
    OptionalParentItem=(Type=Bone,Name="None")
    RotationOffset=(X=-127.0157,Y=4.3906,Z=-18.4663)
    ExecutePin=@n17.ExecutePin
   )
   n20 = RigVMVariableNode(Value=@n19.OutputParam, ExecuteContext=@n18.ExecuteContext) {variable=R thighTwistIn Alpha}
   n21 = &RigUnit_SphericalPoseReader(
    ActiveRegionScaleFactors=(PositiveWidth=1,NegativeWidth=1,PositiveHeight=1,NegativeHeight=1)
    ActiveRegionSize=0.12
    Debug=(bDrawDebug=false,bDraw2D=False,bDrawLocalAxes=False,DebugScale=25,DebugSegments=20,DebugThickness=0.25)
    DriverAxis=(X=-1,Y=0,Z=0)
    DriverItem=(Type=Bone,Name="thigh_l")
    FalloffRegionScaleFactors=(PositiveWidth=1,NegativeWidth=1,PositiveHeight=1,NegativeHeight=1)
    FalloffSize=0.25
    FlipHeightScaling=false
    FlipWidthScaling=false
    OptionalParentItem=(Type=Bone,Name="None")
    RotationOffset=(X=75.9774,Y=-9.6785,Z=-2.3149)
    ExecutePin=@n19.ExecutePin
   )
   n22 = RigVMVariableNode(Value=@n21.OutputParam, ExecuteContext=@n20.ExecuteContext) {variable=L Crouch Strafe Front}
   n23 = &RigUnit_SphericalPoseReader(
    ActiveRegionScaleFactors=(PositiveWidth=1,NegativeWidth=1,PositiveHeight=1,NegativeHeight=1)
    ActiveRegionSize=0.12
    Debug=(bDrawDebug=false,bDraw2D=False,bDrawLocalAxes=False,DebugScale=25,DebugSegments=20,DebugThickness=0.25)
    DriverAxis=(X=1,Y=0,Z=0)
    DriverItem=(Type=Bone,Name="thigh_r")
    FalloffRegionScaleFactors=(PositiveWidth=1,NegativeWidth=1,PositiveHeight=1,NegativeHeight=1)
    FalloffSize=0.25
    FlipHeightScaling=False
    FlipWidthScaling=true
    OptionalParentItem=(Type=Bone,Name="None")
    RotationOffset=(X=-106.852,Y=57.2151,Z=33.6105)
    ExecutePin=@n21.ExecutePin
   )
   n24 = RigVMVariableNode(Value=@n23.OutputParam, ExecuteContext=@n22.ExecuteContext) {variable=R Crouch Strafe Behind}
   n25 = &RigUnit_SphericalPoseReader(
    ActiveRegionScaleFactors=(PositiveWidth=1,NegativeWidth=1,PositiveHeight=1,NegativeHeight=1)
    ActiveRegionSize=0.12
    Debug=(bDrawDebug=false,bDraw2D=False,bDrawLocalAxes=False,DebugScale=25,DebugSegments=20,DebugThickness=0.25)
    DriverAxis=(X=-1,Y=0,Z=0)
    DriverItem=(Type=Bone,Name="thigh_l")
    FalloffRegionScaleFactors=(PositiveWidth=1,NegativeWidth=1,PositiveHeight=1,NegativeHeight=1)
    FalloffSize=0.3
    FlipHeightScaling=false
    FlipWidthScaling=false
    OptionalParentItem=(Type=Bone,Name="None")
    RotationOffset=(X=45.4793,Y=-47.9869,Z=88.1318)
    ExecutePin=@n23.ExecutePin
   )
   n26 = RigVMVariableNode(Value=@n25.OutputParam, ExecuteContext=@n24.ExecuteContext) {variable=L Crouch Strafe Behind}
   n27 = &RigUnit_SphericalPoseReader(
    ActiveRegionScaleFactors=(PositiveWidth=1,NegativeWidth=1,PositiveHeight=1,NegativeHeight=1)
    ActiveRegionSize=0.12
    Debug=(bDrawDebug=false,bDraw2D=False,bDrawLocalAxes=False,DebugScale=25,DebugSegments=20,DebugThickness=0.25)
    DriverAxis=(X=1,Y=0,Z=0)
    DriverItem=(Type=Bone,Name="thigh_r")
    FalloffRegionScaleFactors=(PositiveWidth=1,NegativeWidth=1,PositiveHeight=1,NegativeHeight=1)
    FalloffSize=0.3
    FlipHeightScaling=False
    FlipWidthScaling=true
    OptionalParentItem=(Type=Bone,Name="None")
    RotationOffset=(X=-110.9682,Y=15.2115,Z=-14.8802)
    ExecutePin=@n25.ExecutePin
   )
   n28 = RigVMVariableNode(Value=@n27.OutputParam, ExecuteContext=@n26.ExecuteContext) {variable=R Crouch Strafe Front}
   n29 = &RigUnit_SphericalPoseReader(
    ActiveRegionScaleFactors=(PositiveWidth=1,NegativeWidth=1.2,PositiveHeight=1,NegativeHeight=1)
    ActiveRegionSize=0.13
    Debug=(bDrawDebug=false,bDraw2D=False,bDrawLocalAxes=False,DebugScale=25,DebugSegments=20,DebugThickness=0.25)
    DriverAxis=(X=-1,Y=0,Z=0)
    DriverItem=(Type=Bone,Name="thigh_l")
    FalloffRegionScaleFactors=(PositiveWidth=1,NegativeWidth=1.2,PositiveHeight=1,NegativeHeight=1)
    FalloffSize=0.25
    FlipHeightScaling=false
    FlipWidthScaling=false
    OptionalParentItem=(Type=Bone,Name="None")
    RotationOffset=(X=99.0839,Y=-161.4364,Z=74.6992)
    ExecutePin=@n27.ExecutePin
   )
   n30 = RigVMVariableNode(Value=@n29.OutputParam, ExecuteContext=@n28.ExecuteContext) {variable=L thighSideIn_Bwd Alpha}
   n31 = &RigUnit_SphericalPoseReader(
    ActiveRegionScaleFactors=(PositiveWidth=1,NegativeWidth=1,PositiveHeight=1,NegativeHeight=1)
    ActiveRegionSize=0.12
    Debug=(bDrawDebug=false,bDraw2D=False,bDrawLocalAxes=False,DebugScale=25,DebugSegments=20,DebugThickness=0.25)
    DriverAxis=(X=1,Y=0,Z=0)
    DriverItem=(Type=Bone,Name="thigh_r")
    FalloffRegionScaleFactors=(PositiveWidth=1,NegativeWidth=1,PositiveHeight=1,NegativeHeight=1)
    FalloffSize=0.25
    FlipHeightScaling=False
    FlipWidthScaling=true
    OptionalParentItem=(Type=Bone,Name="None")
    RotationOffset=(X=-108.1179,Y=53.0444,Z=19.1451)
    ExecutePin=@n29.ExecutePin
   )
   n32 = RigVMVariableNode(Value=@n31.OutputParam, ExecuteContext=@n30.ExecuteContext) {variable=R thighSideIn_Fwd Alpha}
   n33 = &RigUnit_SphericalPoseReader(
    ActiveRegionScaleFactors=(PositiveWidth=1,NegativeWidth=1,PositiveHeight=1,NegativeHeight=1)
    ActiveRegionSize=0.12
    Debug=(bDrawDebug=false,bDraw2D=False,bDrawLocalAxes=False,DebugScale=25,DebugSegments=20,DebugThickness=0.25)
    DriverAxis=(X=0,Y=1,Z=0)
    DriverItem=(Type=Bone,Name="thigh_l")
    FalloffRegionScaleFactors=(PositiveWidth=1,NegativeWidth=1,PositiveHeight=1,NegativeHeight=1)
    FalloffSize=0.17
    FlipHeightScaling=false
    FlipWidthScaling=false
    OptionalParentItem=(Type=Bone,Name="None")
    RotationOffset=(X=147.4193,Y=-9.6785,Z=-26.2186)
    ExecutePin=@n31.ExecutePin
   )
   n34 = RigVMVariableNode(Value=@n33.OutputParam, ExecuteContext=@n32.ExecuteContext) {variable=L thighTwistOut Alpha}
   n35 = &RigUnit_SphericalPoseReader(
    ActiveRegionScaleFactors=(PositiveWidth=1,NegativeWidth=1,PositiveHeight=1,NegativeHeight=1)
    ActiveRegionSize=0.08
    Debug=(bDrawDebug=false,bDraw2D=False,bDrawLocalAxes=False,DebugScale=25,DebugSegments=20,DebugThickness=0.25)
    DriverAxis=(X=-1,Y=0,Z=0)
    DriverItem=(Type=Bone,Name="clavicle_r")
    FalloffRegionScaleFactors=(PositiveWidth=1,NegativeWidth=1,PositiveHeight=1,NegativeHeight=1)
    FalloffSize=0.15
    FlipHeightScaling=False
    FlipWidthScaling=False
    OptionalParentItem=(Type=Bone,Name="None")
    RotationOffset=(X=-8.8217,Y=130,Z=0)
    ExecutePin=@n33.ExecutePin
   )
   RigVMVariableNode(Value=@n35.OutputParam, ExecuteContext=@n34.ExecuteContext) {variable=ClavUp}
   n36 = &RigVMFunction_AlphaInterp(
    InterpSpeedDecreasing=10
    InterpSpeedIncreasing=10
    Scale=1
    bClampResult=true
    Value=&RigVMFunction_MathDoubleMul(A=RigVMVariableNode() {variable=L thighSide Alpha}.Value, B=RigVMVariableNode() {variable=Is Playing Emote}.Value).Result
   )
   &RigUnit_ModifyTransforms(
    ItemToModify="((Item=(Type=Bone,Name=\"dyn_dress_side_l_01\"),Transform=(Rotation=(X=0,Y=-0.0872,Z=0,W=0.9962),Translation=(X=-0.9414,Y=0,Z=5.339),Scale3D=(X=1,Y=1,Z=1))),(Item=(Type=Bone,Name=\"dyn_dress_fr_c_01\"),Transform=(Rotation=(X=0.0076,Y=-0.0868,Z=-0.0868,W=0.9924 ...(truncated)"
    WeightMinimum=0
    Weight=@n36.Result
    ExecutePin=@RigVMFunction_Sequence.G
   )
   n37 = &RigVMFunction_AlphaInterp(
    InterpSpeedDecreasing=10
    InterpSpeedIncreasing=10
    Scale=1
    bClampResult=true
    Value=&RigVMFunction_MathDoubleMul(B=RigVMVariableNode() {variable=Is Playing Emote}.Value, A=RigVMVariableNode() {variable=R thighSide Alpha}.Value).Result
   )
   > &RigUnit_ModifyTransforms(
    ItemToModify="((Item=(Type=Bone,Name=\"dyn_dress_side_r_01\"),Transform=(Rotation=(X=0,Y=0.0872,Z=0,W=-0.9962),Translation=(X=0,Y=0,Z=0),Scale3D=(X=1,Y=1,Z=1))),(Item=(Type=Bone,Name=\"dyn_dress_bck_r_01\"),Transform=(Rotation=(X=-0.0076,Y=0.0868,Z=-0.0868,W=-0.9924),Transl ...(truncated)"
    WeightMinimum=0
    Weight=@n37.Result
   )
   n38 = &RigVMFunction_AlphaInterp(
    InterpSpeedDecreasing=10
    InterpSpeedIncreasing=10
    Scale=1
    bClampResult=true
    Value=&RigVMFunction_MathDoubleMul(B=RigVMVariableNode() {variable=Is Playing Emote}.Value, A=RigVMVariableNode() {variable=L thighSideIn_Fwd Alpha}.Value).Result
   )
   > &RigUnit_ModifyTransforms(
    ItemToModify="((Item=(Type=Bone,Name=\"dyn_dress_side_l_01\"),Transform=(Rotation=(X=0,Y=0,Z=0,W=1),Translation=(X=0,Y=0,Z=0),Scale3D=(X=1,Y=1,Z=1))),(Item=(Type=Bone,Name=\"dyn_dress_fr_l_01\"),Transform=(Rotation=(X=-0.0076,Y=0.0868,Z=0.0868,W=-0.9924),Translation=(X=0,Y= ...(truncated)"
    WeightMinimum=0
    Weight=@n38.Result
   )
   n39 = &RigVMFunction_AlphaInterp(
    InterpSpeedDecreasing=10
    InterpSpeedIncreasing=10
    Scale=1
    bClampResult=true
    Value=&RigVMFunction_MathDoubleMul(B=RigVMVariableNode() {variable=Is Playing Emote}.Value, A=RigVMVariableNode() {variable=R Crouch Strafe Behind}.Value).Result
   )
   > &RigUnit_ModifyTransforms(
    ItemToModify="((Item=(Type=Bone,Name=\"dyn_dress_fr_r_01\"),Transform=(Rotation=(X=0.0076,Y=0.0868,Z=0.0868,W=-0.9924),Translation=(X=-0.4923,Y=-2.7921,Z=1.0547),Scale3D=(X=1,Y=1,Z=1))),(Item=(Type=Bone,Name=\"dyn_dress_bck_r_01\"),Transform=(Rotation=(X=0,Y=0,Z=0,W=1),Tran ...(truncated)"
    WeightMinimum=0
    Weight=@n39.Result
   )
   n40 = RigVMVariableNode() {variable=Is Playing Emote}
   n41 = RigVMVariableNode() {variable=L thighSideIn_Fwd Alpha}
   n42 = &RigVMFunction_AlphaInterp(
    InterpSpeedDecreasing=10
    InterpSpeedIncreasing=10
    Scale=1
    bClampResult=true
    Value=&RigVMFunction_MathDoubleMul(B=&RigVMFunction_MathDoubleMul(B=@n40.Value, A=@n41.Value).Result, A=RigVMVariableNode() {variable=R thighForward Alpha}.Value).Result
   )
   > &RigUnit_ModifyTransforms(
    ItemToModify="((Item=(Type=Bone,Name=\"dyn_dress_fr_r_01\"),Transform=(Rotation=(X=0,Y=-0.0872,Z=0,W=0.9962),Translation=(X=0,Y=0,Z=0),Scale3D=(X=1,Y=1,Z=1))),(Item=(Type=Bone,Name=\"dyn_dress_fr_c_01\"),Transform=(Rotation=(X=0,Y=0,Z=-0.0872,W=0.9962),Translation=(X=0,Y=0, ...(truncated)"
    WeightMinimum=0
    Weight=@n42.Result
   )
   n43 = RigVMVariableNode() {variable=R thighBack Alpha}
   n44 = RigVMVariableNode() {variable=L thighTwistOut Alpha}
   n45 = &RigVMFunction_AlphaInterp(
    InterpSpeedDecreasing=10
    InterpSpeedIncreasing=10
    Scale=1
    bClampResult=true
    Value=&RigVMFunction_MathDoubleMul(A=&RigVMFunction_MathDoubleMul(B=@n43.Value, A=@n44.Value).Result, B=RigVMVariableNode() {variable=Is Playing Emote}.Value).Result
   )
   > &RigUnit_ModifyTransforms(
    ItemToModify="((Item=(Type=Bone,Name=\"dyn_dress_fr_r_01\"),Transform=(Rotation=(X=0.0151,Y=-0.0858,Z=0,W=0.9962),Translation=(X=0,Y=0,Z=0),Scale3D=(X=1,Y=1,Z=1))),(Item=(Type=Bone,Name=\"dyn_dress_fr_c_01\"),Transform=(Rotation=(X=0,Y=0,Z=0.0872,W=0.9962),Translation=(X=0, ...(truncated)"
    WeightMinimum=0
    Weight=@n45.Result
   )
   > &RigUnit_ModifyTransforms(
    ItemToModify=((Item=(Type=Bone,Name="dyn_dress_fr_c_01"),Transform=(Rotation=(X=0,Y=-0.0262,Z=0,W=0.9997),Translation=(X=0,Y=0,Z=0),Scale3D=(X=1,Y=1,Z=1))))
    WeightMinimum=0
    Weight=RigVMVariableNode() {variable=R Crouch Strafe Behind}.Value
   )
   &RigUnit_ModifyTransforms(
    ItemToModify="((Item=(Type=Bone,Name=\"dyn_necklace_09\"),Transform=(Rotation=(X=0,Y=0,Z=0,W=-1),Translation=(X=0,Y=0,Z=0),Scale3D=(X=1,Y=1,Z=1))),(Item=(Type=Bone,Name=\"dyn_necklace_010\"),Transform=(Rotation=(X=0,Y=0,Z=0,W=-1),Translation=(X=-0.4196,Y=0,Z=0),Scale3D=(X=1 ...(truncated)"
    WeightMinimum=0
    Weight=RigVMVariableNode() {variable=ClavUp}.Value
    ExecutePin=@RigVMFunction_Sequence.D
   )
   n46 = &RigVMFunction_AlphaInterp(
    InterpSpeedDecreasing=20
    InterpSpeedIncreasing=20
    Scale=1
    bClampResult=true
    Value=RigVMVariableNode() {variable=L thighForward Alpha}.Value
   )
   > &RigUnit_ModifyTransforms(
    ItemToModify="((Item=(Type=Bone,Name=\"dyn_dress_fr_l_01\"),Transform=(Rotation=(X=0,Y=-0.0872,Z=0,W=0.9962),Translation=(X=-0.3099,Y=1.6074,Z=1.7576),Scale3D=(X=1,Y=1,Z=1))),(Item=(Type=Bone,Name=\"dyn_dress_side_l_01\"),Transform=(Rotation=(X=0,Y=0,Z=0,W=1),Translation=(X ...(truncated)"
    WeightMinimum=0
    Weight=@n46.Result
   )
   n47 = &RigVMFunction_AlphaInterp(
    InterpSpeedDecreasing=20
    InterpSpeedIncreasing=20
    Scale=1
    bClampResult=true
    Value=RigVMVariableNode() {variable=R thighForward Alpha}.Value
   )
   > &RigUnit_ModifyTransforms(
    ItemToModify="((Item=(Type=Bone,Name=\"dyn_dress_fr_r_01\"),Transform=(Rotation=(X=0,Y=-0.0872,Z=0,W=0.9962),Translation=(X=-0.2424,Y=0,Z=1.375),Scale3D=(X=1,Y=1,Z=1))),(Item=(Type=Bone,Name=\"dyn_dress_bck_r_01\"),Transform=(Rotation=(X=0,Y=-0.0872,Z=0,W=0.9962),Translatio ...(truncated)"
    WeightMinimum=0
    Weight=@n47.Result
   )
   n48 = &RigVMFunction_AlphaInterp(
    InterpSpeedDecreasing=15
    InterpSpeedIncreasing=15
    Scale=1
    bClampResult=true
    Value=&RigVMFunction_MathDoubleMul(A=RigVMVariableNode() {variable=R thighSide Alpha}.Value, B=RigVMVariableNode() {variable=Not Skydive Alpha}.Value).Result
   )
   > &RigUnit_ModifyTransforms(
    ItemToModify="((Item=(Type=Bone,Name=\"dyn_dress_bck_c_01\"),Transform=(Rotation=(X=0,Y=0,Z=0,W=1),Translation=(X=0,Y=0,Z=0),Scale3D=(X=1,Y=1,Z=1))),(Item=(Type=Bone,Name=\"dyn_dress_bck_r_01\"),Transform=(Rotation=(X=0,Y=0,Z=0,W=1),Translation=(X=0,Y=0,Z=0),Scale3D=(X=1,Y= ...(truncated)"
    WeightMinimum=0
    Weight=@n48.Result
   )
   n49 = &RigVMFunction_AlphaInterp(InterpSpeedDecreasing=45, InterpSpeedIncreasing=45, Scale=1, bClampResult=true, Value=RigVMVariableNode() {variable=ThighSplit L Fwd}.Value)
   > &RigUnit_ModifyTransforms(
    ItemToModify="((Item=(Type=Bone,Name=\"dyn_dress_fr_l_01\"),Transform=(Rotation=(X=0,Y=0,Z=0,W=1),Translation=(X=0,Y=0,Z=0),Scale3D=(X=1,Y=1,Z=1))),(Item=(Type=Bone,Name=\"dyn_dress_fr_r_01\"),Transform=(Rotation=(X=-0.0894,Y=-0.2526,Z=-0.0373,W=0.9627),Translation=(X=0,Y=0 ...(truncated)"
    WeightMinimum=0
    Weight=@n49.Result
   )
   n50 = &RigVMFunction_AlphaInterp(InterpSpeedDecreasing=45, InterpSpeedIncreasing=45, Scale=1, bClampResult=true, Value=RigVMVariableNode() {variable=ThighSplit R Fwd}.Value)
   > &RigUnit_ModifyTransforms(
    ItemToModify="((Item=(Type=Bone,Name=\"dyn_dress_fr_l_01\"),Transform=(Rotation=(X=-0.0872,Y=-0.173,Z=0,W=0.9811),Translation=(X=-0.0392,Y=-1.2812,Z=0.2225),Scale3D=(X=1,Y=1,Z=1))),(Item=(Type=Bone,Name=\"dyn_dress_fr_r_01\"),Transform=(Rotation=(X=0.0076,Y=-0.0868,Z=-0.086 ...(truncated)"
    WeightMinimum=0
    Weight=@n50.Result
   )
   &RigUnit_ParentConstraint(
    AdvancedSettings=(InterpolationType=Average,RotationOrderForFilter=XZY)
    Child=(Type=Bone,Name="dyn_dress_fr_l_01")
    Parents=((Item=(Type=Bone,Name="thigh_l"),Weight=1),(Item=(Type=Bone,Name="thigh_r"),Weight=0))
    Weight=1
    ExecutePin=@RigVMFunction_Sequence.C
   )
   > &RigUnit_ParentConstraint(
    AdvancedSettings=(InterpolationType=Average,RotationOrderForFilter=XZY)
    Child=(Type=Bone,Name="dyn_dress_fr_r_01")
    Parents=((Item=(Type=Bone,Name="thigh_l"),Weight=0),(Item=(Type=Bone,Name="thigh_r"),Weight=1))
    Weight=1
   )
   > &RigUnit_ParentConstraint(
    AdvancedSettings=(InterpolationType=Average,RotationOrderForFilter=XZY)
    Child=(Type=Bone,Name="dyn_dress_fr_l_02")
    Parents=((Item=(Type=Bone,Name="calf_l"),Weight=1),(Item=(Type=Bone,Name="calf_r"),Weight=0))
    Weight=1
   )
   > &RigUnit_ParentConstraint(
    AdvancedSettings=(InterpolationType=Average,RotationOrderForFilter=XZY)
    Child=(Type=Bone,Name="dyn_dress_fr_r_02")
    Parents=((Item=(Type=Bone,Name="calf_l"),Weight=0),(Item=(Type=Bone,Name="calf_r"),Weight=1))
    Weight=1
   )
   > &RigUnit_ParentConstraint(
    AdvancedSettings=(InterpolationType=Average,RotationOrderForFilter=XZY)
    Child=(Type=Bone,Name="dyn_dress_bck_l_01")
    Parents=((Item=(Type=Bone,Name="thigh_l"),Weight=0.8),(Item=(Type=Bone,Name="thigh_r"),Weight=0.2))
    Weight=1
   )
   > &RigUnit_ParentConstraint(
    AdvancedSettings=(InterpolationType=Average,RotationOrderForFilter=XZY)
    Child=(Type=Bone,Name="dyn_dress_bck_r_01")
    Parents=((Item=(Type=Bone,Name="thigh_l"),Weight=0.2),(Item=(Type=Bone,Name="thigh_r"),Weight=0.8))
    Weight=1
   )
   > &RigUnit_ParentConstraint(
    AdvancedSettings=(InterpolationType=Average,RotationOrderForFilter=XZY)
    Child=(Type=Bone,Name="dyn_dress_bck_l_02")
    Parents=((Item=(Type=Bone,Name="calf_l"),Weight=0.9),(Item=(Type=Bone,Name="calf_r"),Weight=0.1))
    Weight=1
   )
   > &RigUnit_ParentConstraint(
    AdvancedSettings=(InterpolationType=Average,RotationOrderForFilter=XZY)
    Child=(Type=Bone,Name="dyn_dress_bck_r_02")
    Parents=((Item=(Type=Bone,Name="calf_l"),Weight=0.1),(Item=(Type=Bone,Name="calf_r"),Weight=0.9))
    Weight=1
   )
   > &RigUnit_ParentConstraint(
    AdvancedSettings=(InterpolationType=Average,RotationOrderForFilter=XZY)
    Child=(Type=Bone,Name="dyn_dress_side_l_01")
    Parents=((Item=(Type=Bone,Name="thigh_l"),Weight=1))
    Weight=1
   )
   > &RigUnit_ParentConstraint(
    AdvancedSettings=(InterpolationType=Average,RotationOrderForFilter=XZY)
    Child=(Type=Bone,Name="dyn_dress_side_r_01")
    Parents=((Item=(Type=Bone,Name="thigh_r"),Weight=1))
    Weight=1
   )
   > &RigUnit_ParentConstraint(
    AdvancedSettings=(InterpolationType=Average,RotationOrderForFilter=XZY)
    Child=(Type=Bone,Name="dyn_dress_side_l_02")
    Parents=((Item=(Type=Bone,Name="calf_l"),Weight=1))
    Weight=1
   )
   > &RigUnit_ParentConstraint(
    AdvancedSettings=(InterpolationType=Average,RotationOrderForFilter=XZY)
    Child=(Type=Bone,Name="dyn_dress_side_r_02")
    Parents=((Item=(Type=Bone,Name="calf_r"),Weight=1))
    Weight=1
   )
   > &RigUnit_ParentConstraint(
    AdvancedSettings=(InterpolationType=Average,RotationOrderForFilter=XZY)
    Child=(Type=Bone,Name="dyn_dress_fr_c_01")
    Parents=((Item=(Type=Bone,Name="dyn_dress_fr_l_01"),Weight=0.5),(Item=(Type=Bone,Name="dyn_dress_fr_r_01"),Weight=0.5))
    Weight=1
   )
   > &RigUnit_ParentConstraint(
    AdvancedSettings=(InterpolationType=Average,RotationOrderForFilter=XZY)
    Child=(Type=Bone,Name="dyn_dress_fr_c_02")
    Parents=((Item=(Type=Bone,Name="dyn_dress_fr_l_02"),Weight=0.5),(Item=(Type=Bone,Name="dyn_dress_fr_r_02"),Weight=0.5))
    Weight=1
   )
   > &RigUnit_ParentConstraint(
    AdvancedSettings=(InterpolationType=Average,RotationOrderForFilter=XZY)
    Child=(Type=Bone,Name="dyn_dress_fr_c_03")
    Parents=((Item=(Type=Bone,Name="dyn_dress_fr_l_03"),Weight=0.5),(Item=(Type=Bone,Name="dyn_dress_fr_r_03"),Weight=0.5))
    Weight=1
   )
   > &RigUnit_ParentConstraint(
    AdvancedSettings=(InterpolationType=Average,RotationOrderForFilter=XZY)
    Child=(Type=Bone,Name="dyn_dress_bck_c_01")
    Parents=((Item=(Type=Bone,Name="dyn_dress_bck_l_01"),Weight=0.5),(Item=(Type=Bone,Name="dyn_dress_bck_r_01"),Weight=0.5))
    Weight=1
   )
   > &RigUnit_ParentConstraint(
    AdvancedSettings=(InterpolationType=Average,RotationOrderForFilter=XZY)
    Child=(Type=Bone,Name="dyn_dress_bck_c_02")
    Parents=((Item=(Type=Bone,Name="dyn_dress_bck_l_02"),Weight=0.5),(Item=(Type=Bone,Name="dyn_dress_bck_r_02"),Weight=0.5))
    Weight=1
   )
   > &RigUnit_ParentConstraint(
    AdvancedSettings=(InterpolationType=Average,RotationOrderForFilter=XZY)
    Child=(Type=Bone,Name="dyn_dress_bck_c_03")
    Parents=((Item=(Type=Bone,Name="dyn_dress_bck_l_03"),Weight=0.5),(Item=(Type=Bone,Name="dyn_dress_bck_r_03"),Weight=0.5))
    Weight=1
   )
   Multiply = &RigVMFunction_MathDoubleMul(A=RigVMVariableNode() {variable=R thighBack Alpha}.Value, B=RigVMVariableNode() {variable=L thighForward Alpha}.Value)
   > RigVMVariableNode(Value=@Multiply.Result) {variable=ThighSplit L Fwd}
   > RigVMVariableNode(Value=&RigVMFunction_MathDoubleMul(A=RigVMVariableNode() {variable=R thighForward Alpha}.Value, B=RigVMVariableNode() {variable=L thighBack Alpha}.Value).Result) {variable=ThighSplit R Fwd}
   n51 = &RigUnit_GetRelativeTransformForItem(Child=(Type=Bone,Name="pelvis"), Parent=(Type=Bone,Name="root"), bChildInitial=False, bParentInitial=false)
   n52 = &RigVMFunction_MathDoubleRemap(
    SourceMaximum=60
    SourceMinimum=80
    TargetMaximum=1
    TargetMinimum=0
    Value=RigVMDispatchNode(Struct=@n51.RelativeTransform.Translation) {template="DISPATCH_RigVMDispatch_BreakStruct(in Struct,out Elements)"}.Elements.Z
   )
   > RigVMVariableNode(Value=@n52.Result) {variable=Low To Ground Alpha}
   > FN_DeformRig_FNC = RigVMFunctionReferenceNode(
    DeformRig_ Curve="(deformRig_l_upperarm_up_alpha_72_#1=False,deformRig_r_upperarm_up_alpha_73_#2=False,deformRig_l_upperarm_fr_alpha_74_#3=False,deformRig_r_upperarm_fr_alpha_75_CB51B8 ...(truncated)"
    DeformRig_Bool="(upperarm_1_#4=True,elbow_3_#5=false,wrist_7_#6=false,groin_8_#7=false,glute_17_#8=false,knee_9_40C408324C ...(truncated)"
    DeformRig_Debug=(upperarm_debug_18_#4=False,wrist_debug_20_#6=False,lat_debug_26_#9=False,groinGlute_debug_27_#7=False)
    DeformRig_Value="(upperarm_up_18_#10=(Rotation=(X=0,Y=-0.1736,Z=0,W=0.9848),Translation=(X=0,Y=5,Z=5),Scale3D=(X=2,Y=2,Z=2)),upperarm_fwd_fr_19_#11=(Rotation=(X=0,Y=0,Z=0,W=1),Translation=(X=0,Y=0,Z=0),Scale3D=(X=1, ...(truncated)"
   )
   n53 = &RigVMFunction_AlphaInterp(InterpSpeedDecreasing=5, InterpSpeedIncreasing=5, Scale=1, bClampResult=true, Value=RigVMVariableNode() {variable=Moving Alpha}.Value)
   n54 = RigVMVariableNode() {variable=R calfBend Alpha}
   n55 = &RigVMFunction_MathDoubleSub(B=1.3, A=RigVMVariableNode() {variable=ThighSplit L Fwd}.Value)
   n56 = &RigVMFunction_MathDoubleAbs(Value=&RigVMFunction_MathDoubleSub(B=1, A=RigVMVariableNode() {variable=L calfBend Alpha}.Value).Result)
   n57 = RigVMVariableNode() {variable=ThighSplit R Fwd}
   n58 = &RigVMFunction_MathDoubleMul(B=RigVMVariableNode() {variable=R thighSideIn_Bwd Alpha}.Value, A=RigVMVariableNode() {variable=L thighSideIn_Fwd Alpha}.Value)
   n59 = &RigVMFunction_MathDoubleMul(B=&RigVMFunction_MathDoubleAbs(Value=&RigVMFunction_MathDoubleSub(B=1, A=@n57.Value).Result).Result, A=@n58.Result)
   n60 = &RigVMFunction_MathDoubleMul(
    B=&RigVMFunction_MathDoubleAbs(Value=&RigVMFunction_MathDoubleSub(B=1.5, A=@n54.Value).Result).Result
    A=&RigVMFunction_MathDoubleMul(
     B=&RigVMFunction_MathDoubleAbs(Value=@n55.Result).Result
     A=&RigVMFunction_MathDoubleMul(B=@n56.Result, A=@n59.Result).Result
    ).Result
   )
   n61 = &RigVMFunction_AlphaInterp(
    InterpSpeedDecreasing=15
    InterpSpeedIncreasing=15
    Scale=1
    bClampResult=true
    Value=&RigVMFunction_MathDoubleMul(A=&RigVMFunction_MathDoubleMul(B=@n53.Result, A=@n60.Result).Result, B=RigVMVariableNode() {variable=Not Skydive Alpha}.Value).Result
   )
   n62 = &RigUnit_ModifyTransforms(
    ItemToModify="((Item=(Type=Bone,Name=\"dyn_dress_fr_r_01\"),Transform=(Rotation=(X=-0.0076,Y=0.0868,Z=-0.0868,W=-0.9924),Translation=(X=0.215,Y=-1.2191,Z=5.7028),Scale3D=(X=1,Y=1,Z=1))),(Item=(Type=Bone,Name=\"dyn_dress_fr_l_01\"),Transform=(Rotation=(X=0,Y=0,Z=0.0872,W=-0. ...(truncated)"
    WeightMinimum=0
    Weight=@n61.Result
    ExecutePin=@RigVMFunction_Sequence.E
   )
   Less = /Script/RigVM.RigVMFunction_MathDoubleLess(B=1, A=@n58.Result) {is=struct RigVMUnitNode, template="Less::Execute(in A,in B,out Result)"}
   RigVMFunction_MathBoolNot = &RigVMFunction_MathBoolNot(Value=RigVMVariableNode() {variable=Is Crouching}.Value)
   RigVMFunction_MathBoolAnd = &RigVMFunction_MathBoolAnd(A=@Less.Result, B=@RigVMFunction_MathBoolNot.Result)
   RigVMFunction_ControlFlowBranch = &RigVMFunction_ControlFlowBranch(ExecuteContext=@n62.ExecutePin, Condition=@RigVMFunction_MathBoolAnd.Result)
   n63 = &RigVMFunction_ControlFlowBranch(ExecuteContext=@RigVMFunction_ControlFlowBranch.Completed, Condition=RigVMVariableNode() {variable=Is Crouching}.Value)
   n64 = &RigVMFunction_MathDoubleAbs(Value=&RigVMFunction_MathDoubleSub(B=1, A=RigVMVariableNode() {variable=ThighSplit L Fwd}.Value).Result)
   n65 = RigVMVariableNode() {variable=R calfBend Alpha}
   n66 = &RigVMFunction_MathDoubleAbs(Value=&RigVMFunction_MathDoubleSub(B=1, A=@n65.Value).Result)
   n67 = RigVMVariableNode() {variable=L calfBend Alpha}
   n68 = &RigVMFunction_AlphaInterp(
    InterpSpeedDecreasing=45
    InterpSpeedIncreasing=45
    Scale=1
    bClampResult=true
    Value=&RigVMFunction_MathDoubleMul(
     B=&RigVMFunction_MathDoubleMul(A=1, B=@n64.Result).Result
     A=&RigVMFunction_MathDoubleMul(B=@n66.Result, A=@n67.Value).Result
    ).Result
   )
   &RigUnit_ModifyTransforms(
    ItemToModify="((Item=(Type=Bone,Name=\"dyn_dress_bck_c_02\"),Transform=(Rotation=(X=0,Y=-0.342,Z=0,W=0.9397),Translation=(X=3.1271,Y=0,Z=1.1382),Scale3D=(X=1,Y=1,Z=1))),(Item=(Type=Bone,Name=\"dyn_dress_bck_l_02\"),Transform=(Rotation=(X=0.0151,Y=-0.1723,Z=-0.0013,W=-0.9849 ...(truncated)"
    WeightMinimum=0
    Weight=@n68.Result
    ExecutePin=@n63.False
   )
   n69 = &RigVMFunction_MathDoubleSub(B=1, A=RigVMVariableNode() {variable=ThighSplit R Fwd}.Value)
   n70 = &RigVMFunction_MathDoubleAbs(Value=&RigVMFunction_MathDoubleSub(B=1, A=@n67.Value).Result)
   n71 = &RigVMFunction_AlphaInterp(
    InterpSpeedDecreasing=45
    InterpSpeedIncreasing=45
    Scale=1
    bClampResult=true
    Value=&RigVMFunction_MathDoubleMul(
     B=&RigVMFunction_MathDoubleAbs(Value=@n69.Result).Result
     A=&RigVMFunction_MathDoubleMul(B=@n70.Result, A=@n65.Value).Result
    ).Result
   )
   > &RigUnit_ModifyTransforms(
    ItemToModify="((Item=(Type=Bone,Name=\"dyn_dress_bck_c_02\"),Transform=(Rotation=(X=-0.0963,Y=0.0116,Z=0.217,W=0.9713),Translation=(X=2.2445,Y=0.8656,Z=1.0251),Scale3D=(X=1,Y=1,Z=1))),(Item=(Type=Bone,Name=\"dyn_dress_bck_r_02\"),Transform=(Rotation=(X=0,Y=0.2588,Z=0,W=0.96 ...(truncated)"
    WeightMinimum=0
    Weight=@n71.Result
   )
   n72 = RigVMVariableNode() {variable=ThighSplit L Fwd}
   n73 = RigVMVariableNode() {variable=ThighSplit R Fwd}
   n74 = &RigVMFunction_MathDoubleMul(
    A=&RigVMFunction_MathDoubleAbs(Value=&RigVMFunction_MathDoubleSub(B=1, A=@n72.Value).Result).Result
    B=&RigVMFunction_MathDoubleAbs(Value=&RigVMFunction_MathDoubleSub(B=1, A=@n73.Value).Result).Result
   )
   n75 = &RigVMFunction_MathDoubleMul(
    A=&RigVMFunction_MathDoubleMul(A=RigVMVariableNode() {variable=L calfBend Alpha}.Value, B=RigVMVariableNode() {variable=R calfBend Alpha}.Value).Result
    B=@n74.Result
   )
   n76 = &RigVMFunction_AlphaInterp(
    InterpSpeedDecreasing=45
    InterpSpeedIncreasing=45
    Scale=1
    bClampResult=true
    Value=&RigVMFunction_MathDoubleMul(A=&RigVMFunction_MathDoubleMul(B=@n74.Result, A=@n75.Result).Result, B=RigVMVariableNode() {variable=Not Skydive Alpha}.Value).Result
   )
   > &RigUnit_ModifyTransforms(
    ItemToModify="((Item=(Type=Bone,Name=\"dyn_dress_bck_c_02\"),Transform=(Rotation=(X=0,Y=-0.0872,Z=0,W=0.9962),Translation=(X=12.99,Y=0,Z=-1.4954),Scale3D=(X=1,Y=1,Z=1))),(Item=(Type=Bone,Name=\"dyn_dress_bck_r_02\"),Transform=(Rotation=(X=0,Y=0,Z=0,W=1),Translation=(X=7.053 ...(truncated)"
    WeightMinimum=0
    Weight=@n76.Result
   )
   n77 = &RigVMFunction_MathDoubleMul(A=RigVMVariableNode() {variable=L Crouch Strafe Front}.Value, B=RigVMVariableNode() {variable=R Crouch Strafe Behind}.Value)
   n78 = &RigVMFunction_AlphaInterp(InterpSpeedDecreasing=30, InterpSpeedIncreasing=30, Scale=1, bClampResult=true, Value=@n77.Result)
   &RigUnit_ModifyTransforms(
    ItemToModify="((Item=(Type=Bone,Name=\"dyn_dress_fr_r_01\"),Transform=(Rotation=(X=-0.0868,Y=0.2578,Z=-0.0076,W=-0.9623),Translation=(X=0,Y=0,Z=0),Scale3D=(X=1,Y=1,Z=1))),(Item=(Type=Bone,Name=\"dyn_dress_fr_l_01\"),Transform=(Rotation=(X=0,Y=0,Z=0,W=1),Translation=(X=0,Y=0 ...(truncated)"
    WeightMinimum=0
    Weight=@n78.Result
    ExecutePin=@n63.True
   )
   n79 = &RigVMFunction_AlphaInterp(
    InterpSpeedDecreasing=30
    InterpSpeedIncreasing=30
    Scale=1
    bClampResult=true
    Value=&RigVMFunction_MathDoubleMul(A=RigVMVariableNode() {variable=L Crouch Strafe Behind}.Value, B=RigVMVariableNode() {variable=R Crouch Strafe Front}.Value).Result
   )
   > &RigUnit_ModifyTransforms(
    ItemToModify="((Item=(Type=Bone,Name=\"dyn_dress_fr_r_01\"),Transform=(Rotation=(X=0,Y=0,Z=0,W=1),Translation=(X=0,Y=0,Z=0),Scale3D=(X=1,Y=1,Z=1))),(Item=(Type=Bone,Name=\"dyn_dress_fr_l_01\"),Transform=(Rotation=(X=-0.0842,Y=0.2578,Z=-0.0226,W=-0.9623),Translation=(X=-0.12 ...(truncated)"
    WeightMinimum=0
    Weight=@n79.Result
   )
   > &RigUnit_ModifyTransforms(
    ItemToModify="((Item=(Type=Bone,Name=\"dyn_dress_bck_c_02\"),Transform=(Rotation=(X=0,Y=0.0872,Z=0,W=-0.9962),Translation=(X=10.8328,Y=0.0001,Z=5.3787),Scale3D=(X=1,Y=1,Z=1))),(Item=(Type=Bone,Name=\"dyn_dress_bck_r_02\"),Transform=(Rotation=(X=-0.0151,Y=0.0013,Z=0.1723,W=- ...(truncated)"
    WeightMinimum=0
    Weight=RigVMVariableNode() {variable=Low To Ground Alpha}.Value
   )
   n80 = RigVMVariableNode() {variable=L Crouch Strafe Behind}
   n81 = RigVMVariableNode() {variable=R Crouch Strafe Front}
   n82 = &RigVMFunction_AlphaInterp(
    InterpSpeedDecreasing=45
    InterpSpeedIncreasing=45
    Scale=1
    bClampResult=true
    Value=&RigVMFunction_MathDoubleMul(B=&RigVMFunction_MathDoubleMul(A=@n80.Value, B=@n81.Value).Result, A=RigVMVariableNode() {variable=Slope Roll Down}.Value).Result
   )
   > &RigUnit_ModifyTransforms(
    ItemToModify="((Item=(Type=Bone,Name=\"dyn_dress_bck_c_02\"),Transform=(Rotation=(X=0.0261,Y=0.0026,Z=0.0151,W=0.9995),Translation=(X=-3.6621,Y=-2.0847,Z=8.4252),Scale3D=(X=1,Y=1,Z=1))),(Item=(Type=Bone,Name=\"dyn_dress_bck_r_02\"),Transform=(Rotation=(X=-0.1294,Y=0,Z=0.224 ...(truncated)"
    WeightMinimum=0
    Weight=@n82.Result
   )
   n83 = RigVMVariableNode() {variable=R Crouch Strafe Behind}
   n84 = RigVMVariableNode() {variable=L Crouch Strafe Front}
   n85 = &RigVMFunction_AlphaInterp(
    InterpSpeedDecreasing=45
    InterpSpeedIncreasing=45
    Scale=1
    bClampResult=true
    Value=&RigVMFunction_MathDoubleMul(B=&RigVMFunction_MathDoubleMul(B=@n83.Value, A=@n84.Value).Result, A=RigVMVariableNode() {variable=Slope Roll Down}.Value).Result
   )
   > &RigUnit_ModifyTransforms(
    ItemToModify="((Item=(Type=Bone,Name=\"dyn_dress_bck_c_02\"),Transform=(Rotation=(X=-0.0076,Y=-0.421,Z=-0.0868,W=0.9029),Translation=(X=0,Y=0,Z=0),Scale3D=(X=1,Y=1,Z=1))),(Item=(Type=Bone,Name=\"dyn_dress_bck_r_02\"),Transform=(Rotation=(X=0,Y=0.1736,Z=0,W=-0.9848),Translat ...(truncated)"
    WeightMinimum=0
    Weight=@n85.Result
   )
   n86 = RigVMVariableNode() {variable=R calfBend Alpha}
   n87 = RigVMVariableNode() {variable=Slope Roll Down}
   n88 = &RigVMFunction_AlphaInterp(
    InterpSpeedDecreasing=30
    InterpSpeedIncreasing=30
    Scale=1
    bClampResult=true
    Value=&RigVMFunction_MathDoubleMul(A=&RigVMFunction_MathDoubleMul(B=@n86.Value, A=@n87.Value).Result, B=RigVMVariableNode() {variable=Moving Alpha}.Value).Result
   )
   > &RigUnit_ModifyTransforms(
    ItemToModify="((Item=(Type=Bone,Name=\"dyn_dress_bck_c_02\"),Transform=(Rotation=(X=0,Y=0,Z=0,W=1),Translation=(X=0,Y=0,Z=0),Scale3D=(X=1,Y=1,Z=1))),(Item=(Type=Bone,Name=\"dyn_dress_bck_r_02\"),Transform=(Rotation=(X=0,Y=0,Z=0,W=1),Translation=(X=0,Y=0,Z=0),Scale3D=(X=1,Y= ...(truncated)"
    WeightMinimum=0
    Weight=@n88.Result
   )
   n89 = RigVMVariableNode() {variable=Slope Roll Down}
   n90 = RigVMVariableNode() {variable=L calfBend Alpha}
   n91 = &RigVMFunction_AlphaInterp(
    InterpSpeedDecreasing=30
    InterpSpeedIncreasing=30
    Scale=1
    bClampResult=true
    Value=&RigVMFunction_MathDoubleMul(A=&RigVMFunction_MathDoubleMul(A=@n89.Value, B=@n90.Value).Result, B=RigVMVariableNode() {variable=Moving Alpha}.Value).Result
   )
   > &RigUnit_ModifyTransforms(
    ItemToModify="((Item=(Type=Bone,Name=\"dyn_dress_bck_c_02\"),Transform=(Rotation=(X=0,Y=0,Z=0,W=1),Translation=(X=0,Y=0,Z=0),Scale3D=(X=1,Y=1,Z=1))),(Item=(Type=Bone,Name=\"dyn_dress_bck_r_02\"),Transform=(Rotation=(X=0,Y=0,Z=0,W=1),Translation=(X=0,Y=0,Z=0),Scale3D=(X=1,Y= ...(truncated)"
    WeightMinimum=0
    Weight=@n91.Result
   )
   n92 = &RigVMFunction_MathDoubleAbs(Value=&RigVMFunction_MathDoubleSub(B=1, A=RigVMVariableNode() {variable=ThighSplit L Fwd}.Value).Result)
   n93 = &RigVMFunction_MathDoubleMul(A=RigVMVariableNode() {variable=L Crouch Strafe Front}.Value, B=RigVMVariableNode() {variable=R Crouch Strafe Behind}.Value)
   n94 = &RigVMFunction_AlphaInterp(
    InterpSpeedDecreasing=30
    InterpSpeedIncreasing=30
    Scale=1
    bClampResult=true
    Value=&RigVMFunction_MathDoubleMul(
     B=&RigVMFunction_MathDoubleMul(A=@n92.Result, B=@n93.Result).Result
     A=RigVMVariableNode() {variable=Slope Pitch Sideways}.Value
    ).Result
   )
   > &RigUnit_ModifyTransforms(
    ItemToModify="((Item=(Type=Bone,Name=\"dyn_dress_bck_c_02\"),Transform=(Rotation=(X=0,Y=0,Z=0,W=1),Translation=(X=0,Y=0,Z=0),Scale3D=(X=1,Y=1,Z=1))),(Item=(Type=Bone,Name=\"dyn_dress_bck_r_02\"),Transform=(Rotation=(X=0,Y=0,Z=0.0872,W=-0.9962),Translation=(X=0,Y=0,Z=0),Scal ...(truncated)"
    WeightMinimum=0
    Weight=@n94.Result
   )
   n95 = RigVMVariableNode() {variable=L Crouch Strafe Behind}
   n96 = RigVMVariableNode() {variable=R Crouch Strafe Front}
   n97 = &RigVMFunction_AlphaInterp(
    InterpSpeedDecreasing=45
    InterpSpeedIncreasing=45
    Scale=1
    bClampResult=true
    Value=&RigVMFunction_MathDoubleMul(B=&RigVMFunction_MathDoubleMul(A=@n95.Value, B=@n96.Value).Result, A=RigVMVariableNode() {variable=Slope Pitch Sideways}.Value).Result
   )
   > &RigUnit_ModifyTransforms(
    ItemToModify="((Item=(Type=Bone,Name=\"dyn_dress_bck_c_02\"),Transform=(Rotation=(X=0,Y=0,Z=0,W=1),Translation=(X=0,Y=0,Z=0),Scale3D=(X=1,Y=1,Z=1))),(Item=(Type=Bone,Name=\"dyn_dress_bck_r_02\"),Transform=(Rotation=(X=0,Y=0,Z=0,W=1),Translation=(X=0,Y=0,Z=0),Scale3D=(X=1,Y= ...(truncated)"
    WeightMinimum=0
    Weight=@n97.Result
   )
   > &RigUnit_ModifyTransforms(
    ItemToModify="((Item=(Type=Bone,Name=\"dyn_dress_side_l_01\"),Transform=(Rotation=(X=-0.0872,Y=0,Z=0,W=0.9962),Translation=(X=0,Y=0.1419,Z=0.8048),Scale3D=(X=1,Y=1,Z=1))),(Item=(Type=Bone,Name=\"dyn_dress_side_r_01\"),Transform=(Rotation=(X=0,Y=-0.0872,Z=0,W=0.9962),Transla ...(truncated)"
    WeightMinimum=0
    Weight=RigVMVariableNode() {variable=Slope Pitch Sideways}.Value
   )
   n98 = &RigVMFunction_AlphaInterp(
    InterpSpeedDecreasing=25
    InterpSpeedIncreasing=25
    Scale=1
    bClampResult=true
    Value=&RigVMFunction_MathDoubleMul(A=RigVMVariableNode() {variable=L thighSide Alpha}.Value, B=RigVMVariableNode() {variable=R thighBack Alpha}.Value).Result
   )
   &RigUnit_ModifyTransforms(
    ItemToModify="((Item=(Type=Bone,Name=\"dyn_dress_fr_l_01\"),Transform=(Rotation=(X=0,Y=0.0872,Z=0,W=-0.9962),Translation=(X=0,Y=0,Z=0),Scale3D=(X=1,Y=1,Z=1))),(Item=(Type=Bone,Name=\"dyn_dress_bck_c_01\"),Transform=(Rotation=(X=0,Y=0,Z=0,W=1),Translation=(X=0,Y=0,Z=0),Scale ...(truncated)"
    WeightMinimum=0
    Weight=@n98.Result
    ExecutePin=@n63.Completed
   )
   n99 = &RigVMFunction_AlphaInterp(
    InterpSpeedDecreasing=25
    InterpSpeedIncreasing=25
    Scale=1
    bClampResult=true
    Value=&RigVMFunction_MathDoubleMul(A=RigVMVariableNode() {variable=L thighSide Alpha}.Value, B=RigVMVariableNode() {variable=R thighForward Alpha}.Value).Result
   )
   > &RigUnit_ModifyTransforms(
    ItemToModify="((Item=(Type=Bone,Name=\"dyn_dress_bck_c_01\"),Transform=(Rotation=(X=0,Y=-0.0872,Z=0,W=0.9962),Translation=(X=-4.2618,Y=0.0001,Z=-0.7515),Scale3D=(X=1,Y=1,Z=1))),(Item=(Type=Bone,Name=\"dyn_dress_bck_l_01\"),Transform=(Rotation=(X=0,Y=0.2588,Z=0,W=-0.9659),Tr ...(truncated)"
    WeightMinimum=0
    Weight=@n99.Result
   )
   n100 = &RigVMFunction_MathDoubleSub(B=1, A=RigVMVariableNode() {variable=Not Skydive Alpha}.Value)
   n101 = &RigVMFunction_AlphaInterp(
    InterpSpeedDecreasing=25
    InterpSpeedIncreasing=25
    Scale=1
    bClampResult=true
    Value=&RigVMFunction_MathDoubleMul(B=&RigVMFunction_MathDoubleAbs(Value=@n100.Result).Result, A=RigVMVariableNode() {variable=L thighBack Alpha}.Value).Result
   )
   > &RigUnit_ModifyTransforms(
    ItemToModify="((Item=(Type=Bone,Name=\"dyn_dress_bck_c_01\"),Transform=(Rotation=(X=0,Y=-0.0872,Z=0,W=0.9962),Translation=(X=0,Y=0,Z=0),Scale3D=(X=1,Y=1,Z=1))),(Item=(Type=Bone,Name=\"dyn_dress_bck_l_01\"),Transform=(Rotation=(X=0,Y=0.0872,Z=0,W=-0.9962),Translation=(X=0,Y= ...(truncated)"
    WeightMinimum=0
    Weight=@n101.Result
   )
   n102 = &RigVMFunction_MathDoubleSub(B=1, A=RigVMVariableNode() {variable=Not Skydive Alpha}.Value)
   n103 = &RigVMFunction_AlphaInterp(
    InterpSpeedDecreasing=25
    InterpSpeedIncreasing=25
    Scale=1
    bClampResult=true
    Value=&RigVMFunction_MathDoubleMul(B=&RigVMFunction_MathDoubleAbs(Value=@n102.Result).Result, A=RigVMVariableNode() {variable=R thighBack Alpha}.Value).Result
   )
   > &RigUnit_ModifyTransforms(
    ItemToModify="((Item=(Type=Bone,Name=\"dyn_dress_bck_c_01\"),Transform=(Rotation=(X=0,Y=-0.0872,Z=0,W=0.9962),Translation=(X=0,Y=0,Z=0),Scale3D=(X=1,Y=1,Z=1))),(Item=(Type=Bone,Name=\"dyn_dress_bck_l_01\"),Transform=(Rotation=(X=0,Y=0,Z=0,W=1),Translation=(X=0,Y=0,Z=0),Scal ...(truncated)"
    WeightMinimum=0
    Weight=@n103.Result
   )
   > &RigUnit_ModifyTransforms(
    ItemToModify="((Item=(Type=Bone,Name=\"dyn_dress_bck_c_01\"),Transform=(Rotation=(X=0,Y=-0.0872,Z=0,W=0.9962),Translation=(X=0,Y=0,Z=0),Scale3D=(X=1,Y=1,Z=1))),(Item=(Type=Bone,Name=\"dyn_dress_bck_l_01\"),Transform=(Rotation=(X=0,Y=0.0872,Z=0,W=-0.9962),Translation=(X=0,Y= ...(truncated)"
    WeightMinimum=0
    Weight=RigVMVariableNode() {variable=Moving Up Aggressively Alpha}.Value
   )
   n104 = &RigVMFunction_MathDoubleSub(B=1, A=RigVMVariableNode() {variable=Low To Ground Alpha}.Value)
   n105 = &RigVMFunction_MathDoubleSub(B=1, A=@n61.Result)
   n106 = &RigVMFunction_MathDoubleAbs(Value=&RigVMFunction_MathDoubleSub(B=1, A=RigVMVariableNode() {variable=ThighSplit R Fwd}.Value).Result)
   n107 = RigVMVariableNode() {variable=R calfBend Alpha}
   n108 = &RigVMFunction_MathDoubleSub(B=1, A=RigVMVariableNode() {variable=L thighForward Alpha}.Value)
   n109 = RigVMVariableNode() {variable=L thighSideIn_Bwd Alpha}
   n110 = RigVMVariableNode() {variable=R thighSideIn_Fwd Alpha}
   n111 = &RigVMFunction_MathDoubleMul(
    B=&RigVMFunction_MathDoubleAbs(Value=&RigVMFunction_MathDoubleSub(B=1, A=@n107.Value).Result).Result
    A=&RigVMFunction_MathDoubleMul(
     B=&RigVMFunction_MathDoubleAbs(Value=@n108.Result).Result
     A=&RigVMFunction_MathDoubleMul(A=@n109.Value, B=@n110.Value).Result
    ).Result
   )
   n112 = &RigVMFunction_MathDoubleMul(
    A=&RigVMFunction_MathDoubleMul(
     B=&RigVMFunction_MathDoubleAbs(Value=@n105.Result).Result
     A=&RigVMFunction_MathDoubleMul(B=@n106.Result, A=@n111.Result).Result
    ).Result
    B=RigVMVariableNode() {variable=Moving Alpha}.Value
   )
   n113 = RigVMVariableNode() {variable=Not Skydive Alpha}
   n114 = &RigVMFunction_AlphaInterp(
    InterpSpeedDecreasing=15
    InterpSpeedIncreasing=15
    Scale=1
    bClampResult=true
    Value=&RigVMFunction_MathDoubleMul(
     B=&RigVMFunction_MathDoubleAbs(Value=@n104.Result).Result
     A=&RigVMFunction_MathDoubleMul(A=@n112.Result, B=@n113.Value).Result
    ).Result
   )
   &RigUnit_ModifyTransforms(
    ItemToModify="((Item=(Type=Bone,Name=\"dyn_dress_fr_r_01\"),Transform=(Rotation=(X=0,Y=0,Z=0,W=1),Translation=(X=0,Y=0,Z=0),Scale3D=(X=1,Y=1,Z=1))),(Item=(Type=Bone,Name=\"dyn_dress_fr_l_01\"),Transform=(Rotation=(X=0.0076,Y=-0.2578,Z=0.0868,W=0.9623),Translation=(X=0,Y=0,Z ...(truncated)"
    WeightMinimum=0
    Weight=@n114.Result
    ExecutePin=@RigVMFunction_ControlFlowBranch.True
   )
   n115 = &RigVMFunction_AlphaInterp(
    InterpSpeedDecreasing=10
    InterpSpeedIncreasing=10
    Scale=1
    bClampResult=true
    Value=&RigVMFunction_MathDoubleMul(A=RigVMVariableNode() {variable=L thighTwistIn Alpha}.Value, B=RigVMVariableNode() {variable=R thighTwistIn Alpha}.Value).Result
   )
   &RigUnit_ModifyTransforms(
    ItemToModify="((Item=(Type=Bone,Name=\"dyn_dress_fr_c_01\"),Transform=(Rotation=(X=0,Y=-0.0872,Z=0,W=0.9962),Translation=(X=-1.154,Y=-0.0001,Z=6.5449),Scale3D=(X=1,Y=1,Z=1))),(Item=(Type=Bone,Name=\"dyn_dress_fr_l_01\"),Transform=(Rotation=(X=0,Y=0,Z=0,W=-1),Translation=(X= ...(truncated)"
    WeightMinimum=0
    Weight=@n115.Result
    ExecutePin=@RigVMFunction_Sequence.F
   )
   n116 = RigVMVariableNode() {variable=L thighTwistIn Alpha}
   n117 = RigVMVariableNode() {variable=R thighTwistIn Alpha}
   n118 = &RigVMFunction_AlphaInterp(
    InterpSpeedDecreasing=20
    InterpSpeedIncreasing=20
    Scale=1
    bClampResult=true
    Value=&RigVMFunction_MathDoubleSub(A=@n116.Value, B=@n117.Value).Result
   )
   > &RigUnit_ModifyTransforms(
    ItemToModify="((Item=(Type=Bone,Name=\"dyn_dress_fr_c_01\"),Transform=(Rotation=(X=0.0076,Y=-0.0868,Z=-0.0868,W=0.9924),Translation=(X=0,Y=0,Z=0),Scale3D=(X=1,Y=1,Z=1))),(Item=(Type=Bone,Name=\"dyn_dress_fr_l_01\"),Transform=(Rotation=(X=0,Y=0.0872,Z=0,W=-0.9962),Translatio ...(truncated)"
    WeightMinimum=0
    Weight=@n118.Result
   )
   n119 = &RigVMFunction_AlphaInterp(
    InterpSpeedDecreasing=20
    InterpSpeedIncreasing=20
    Scale=1
    bClampResult=true
    Value=&RigVMFunction_MathDoubleSub(B=@n116.Value, A=@n117.Value).Result
   )
   > &RigUnit_ModifyTransforms(
    ItemToModify="((Item=(Type=Bone,Name=\"dyn_dress_fr_c_01\"),Transform=(Rotation=(X=0.0151,Y=-0.0858,Z=0,W=0.9962),Translation=(X=0,Y=0,Z=0),Scale3D=(X=1,Y=1,Z=1))),(Item=(Type=Bone,Name=\"dyn_dress_fr_l_01\"),Transform=(Rotation=(X=0,Y=0,Z=0,W=1),Translation=(X=0,Y=0,Z=0),S ...(truncated)"
    WeightMinimum=0
    Weight=@n119.Result
   )
   &RigVMFunction_AlphaInterp(
    InterpSpeedDecreasing=10
    InterpSpeedIncreasing=10
    Scale=1
    bClampResult=true
    Value=&RigVMFunction_MathDoubleMul(A=@n116.Value, B=@n117.Value).Result
   )
   n120 = RigVMVariableNode() {variable=ThighSplit L Fwd}
   &RigVMFunction_MathDoubleMul(
    A=&RigVMFunction_MathDoubleAbs(Value=&RigVMFunction_MathDoubleSub(B=1, A=@n120.Value).Result).Result
    B=@n77.Result
   )
   RigVMFunction_MathBoolToFloat = &RigVMFunction_MathBoolToFloat(Value=RigVMVariableNode() {variable=Is On Grind Rail}.Value)
   Subtract = &RigVMFunction_MathDoubleSub(B=1, A=@RigVMFunction_MathBoolToFloat.Result)
   &RigVMFunction_MathDoubleMul(B=&RigVMFunction_MathDoubleAbs(Value=@Subtract.Result).Result, A=@n52.Result)
   Absolute = &RigVMFunction_MathDoubleAbs(Value=&RigVMFunction_MathQuaternionToEuler(RotationOrder=ZYX, Value=@n51.RelativeTransform.Rotation).Result.Y)
   &RigVMFunction_MathDoubleMul(
    B=@n52.Result
    A=&RigVMFunction_MathDoubleRemap(SourceMaximum=69, SourceMinimum=80, TargetMaximum=1, TargetMinimum=0, Value=@Absolute.Result).Result
   )
   n121 = &RigUnit_GetRelativeTransformForItem(Child=(Type=Bone,Name="calf_r"), Parent=(Type=Bone,Name="thigh_r"), bChildInitial=False, bParentInitial=False)
   RigVMVariableNode(
    Value=&RigVMFunction_MathDoubleRemap(
     SourceMaximum=100
     SourceMinimum=45
     TargetMaximum=1
     TargetMinimum=0
     Value=&RigVMFunction_MathQuaternionToEuler(RotationOrder=ZYX, Value=@n121.RelativeTransform.Rotation).Result.Z
    ).Result
    ExecuteContext=@n123.ExecuteContext
   ) {variable=R calfBend Alpha}
   n122 = &RigUnit_GetRelativeTransformForItem(Child=(Type=Bone,Name="calf_l"), Parent=(Type=Bone,Name="thigh_l"), bChildInitial=False, bParentInitial=False)
   RigVMFunction_MathQuaternionToEuler = &RigVMFunction_MathQuaternionToEuler(RotationOrder=ZYX, Value=@n122.RelativeTransform.Rotation)
   Remap = &RigVMFunction_MathDoubleRemap(SourceMaximum=100, SourceMinimum=45, TargetMaximum=1, TargetMinimum=0, Value=@RigVMFunction_MathQuaternionToEuler.Result.Z)
   n123 = RigVMVariableNode(Value=@Remap.Result, ExecuteContext=@n35.ExecutePin) {variable=L calfBend Alpha}
  ]}
]
hierarchy [
 root {loc=<0 0 0>, rot=<0 0 0 1>}
 pelvis {parent=root, loc=<0 -0.5301 98.0571>, rot=<0.0127 0.707 -0.0127 -0.707>}
 spine_01 {parent=pelvis, loc=<6.1208 -0.9328 0>, rot=<0 0 0.034 -0.9994>}
 spine_02 {parent=spine_01, loc=<6.5997 -0.2095 0>, rot=<0 0 0.0238 -0.9997>}
 spine_03 {parent=spine_02, loc=<6.4914 -0.1237 0>, rot=<0 0 0.0149 -0.9999>}
 spine_04 {parent=spine_03, loc=<6.9598 0.0794 0>, rot=<0 0 -0.0132 -0.9999>}
 spine_05 {parent=spine_04, loc=<6.9987 0 0>, rot=<0 0 -0.0015 -1>}
 neck_01 {parent=spine_05, loc=<11.3138 0.1529 0>, rot=<0 0 -0.2003 -0.9797>}
 neck_02 {parent=neck_01, loc=<4.5627 0 0>, rot=<0 0 0.0504 -0.9987>}
 head {parent=neck_02, loc=<5.3914 0 0>, rot=<0 0 0.115 -0.9934>}
 clavicle_l {parent=spine_05, loc=<5.4831 -0.1053 -3.2946>, rot=<0.0465 -0.7289 -0.009 -0.6829>}
 upperarm_l {parent=clavicle_l, loc=<9.8658 -1.2136 -0.4105>, rot=<0.0164 -0.3245 -0.0269 -0.9454>}
 lowerarm_l {parent=upperarm_l, loc=<26.465 -0.8345 0.259>, rot=<0.0036 0.0161 -0.2298 -0.9731>}
 hand_l {parent=lowerarm_l, loc=<22.7355 -0.3203 -0.6105>, rot=<-0.6784 -0.0223 0.0735 -0.7307>}
 thumb_01_l {parent=hand_l, loc=<1.3118 -0.3191 -2.459>, rot=<0.6546 -0.4042 0.0664 -0.6354>}
 thumb_02_l {parent=thumb_01_l, loc=<5.9336 -0.0448 0.0401>, rot=<-0.0037 0.0213 0.298 -0.9543>}
 thumb_03_l {parent=thumb_02_l, loc=<2.2223 0.3682 -0.0907>, rot=<0.0118 -0.013 0.0066 -0.9998>}
 index_metacarpal_l {parent=hand_l, loc=<1.6535 0.0183 -2.96>, rot=<0.0158 -0.0124 -0.0173 -0.9996>}
 index_01_l {parent=index_metacarpal_l, loc=<8.3008 0.0357 -0.0524>, rot=<-0.0044 0.0485 0.0924 -0.9945>}
 index_02_l {parent=index_01_l, loc=<3.3587 -0.3004 0.0452>, rot=<-0.002 0.0027 0.3225 -0.9466>}
 index_03_l {parent=index_02_l, loc=<2.5578 -0.123 -0.0252>, rot=<-0.002 0.0067 0.2558 -0.9667>}
 middle_metacarpal_l {parent=hand_l, loc=<2.1856 -0.1725 -0.5984>, rot=<0.006 0.0168 -0.0329 -0.9993>}
 middle_01_l {parent=middle_metacarpal_l, loc=<7.3275 -0.2447 0.0129>, rot=<-0.0208 0.0042 0.0858 -0.9961>}
 middle_02_l {parent=middle_01_l, loc=<4.0123 0.1401 0.1175>, rot=<-0.0218 0.0258 0.3512 -0.9357>}
 middle_03_l {parent=middle_02_l, loc=<2.6077 0.0534 0.1373>, rot=<0.0082 0.0188 0.1708 -0.9851>}
 ring_metacarpal_l {parent=hand_l, loc=<2.2074 -0.3625 1.3638>, rot=<-0.0284 0.0722 -0.0458 -0.9959>}
 ring_01_l {parent=ring_metacarpal_l, loc=<7.0295 -0.6197 -0.1446>, rot=<-0.0489 -0.0252 0.1892 -0.9804>}
 ring_02_l {parent=ring_01_l, loc=<3.6174 0.4647 -0.1027>, rot=<0 0 0.2806 -0.9598>}
 ring_03_l {parent=ring_02_l, loc=<2.4152 0.114 0.0199>, rot=<0 0 0.1189 -0.9929>}
 pinky_metacarpal_l {parent=hand_l, loc=<1.7616 -0.9753 3.3814>, rot=<-0.0397 0.0814 -0.069 -0.9935>}
 pinky_01_l {parent=pinky_metacarpal_l, loc=<7.1709 -0.4831 -0.1094>, rot=<-0.0654 -0.0035 0.2242 -0.9723>}
 pinky_02_l {parent=pinky_01_l, loc=<2.8274 0.0893 -0.1043>, rot=<0.0037 0.0011 0.2854 -0.9584>}
 pinky_03_l {parent=pinky_02_l, loc=<2.2046 0.0307 -0.0078>, rot=<-0.0002 0.0098 0.1445 -0.9895>}
 lowerarm_twist_02_l {parent=lowerarm_l, loc=<10.6707 0 0>, rot=<0 0 0 1>}
 lowerarm_twist_01_l {parent=lowerarm_l, loc=<18.316 0 0>, rot=<0 0 0 1>}
 upperarm_twist_01_l {parent=upperarm_l, loc=<7.2135 -0.1134 0>, rot=<0 0 0 1>}
 deform_upperarm_bk_base_l {parent=upperarm_twist_01_l, loc=<-6.4053 -4.3746 0.314>, rot=<0 0 0 -1>}
 deform_upperarm_bk_l {parent=deform_upperarm_bk_base_l, loc=<0 0 0>, rot=<0 0 0 1>}
 deform_upperarm_fr_base_l {parent=upperarm_twist_01_l, loc=<-5.6312 5.092 0.3354>, rot=<0 0 0 -1>}
 deform_upperarm_fr_l {parent=deform_upperarm_fr_base_l, loc=<0 0 0>, rot=<0 0 0 1>}
 deform_upperarm_up_base_l {parent=upperarm_twist_01_l, loc=<-8.891 0.4825 3.9517>, rot=<0 0 0 -1>}
 deform_upperarm_up_l {parent=deform_upperarm_up_base_l, loc=<0 0 0>, rot=<0 0 0 1>}
 upperarm_twist_02_l {parent=upperarm_l, loc=<19.7998 -0.1134 0>, rot=<0 0 0 1>}
 clavicle_r {parent=spine_05, loc=<5.4831 -0.1053 3.2946>, rot=<0.7289 0.0465 -0.6829 0.009>}
 upperarm_r {parent=clavicle_r, loc=<-9.8658 1.2136 0.4105>, rot=<0.0164 -0.3245 -0.0269 -0.9454>}
 lowerarm_r {parent=upperarm_r, loc=<-26.465 0.8345 -0.259>, rot=<0.0036 0.0161 -0.2298 -0.9731>}
 hand_r {parent=lowerarm_r, loc=<-22.7355 0.3203 0.6106>, rot=<-0.6784 -0.0223 0.0735 -0.7307>}
 thumb_01_r {parent=hand_r, loc=<-1.3118 0.3191 2.459>, rot=<0.6546 -0.4042 0.0664 -0.6354>}
 thumb_02_r {parent=thumb_01_r, loc=<-5.9336 0.0448 -0.0401>, rot=<-0.0037 0.0213 0.298 -0.9543>}
 thumb_03_r {parent=thumb_02_r, loc=<-2.2223 -0.3682 0.0907>, rot=<0.0118 -0.013 0.0066 -0.9998>}
 index_metacarpal_r {parent=hand_r, loc=<-1.6535 -0.0183 2.96>, rot=<0.0158 -0.0124 -0.0173 -0.9996>}
 index_01_r {parent=index_metacarpal_r, loc=<-8.3008 -0.0357 0.0524>, rot=<-0.0044 0.0485 0.0924 -0.9945>}
 index_02_r {parent=index_01_r, loc=<-3.3587 0.3004 -0.0452>, rot=<-0.002 0.0027 0.3225 -0.9466>}
 index_03_r {parent=index_02_r, loc=<-2.5578 0.123 0.0252>, rot=<-0.002 0.0067 0.2558 -0.9667>}
 middle_metacarpal_r {parent=hand_r, loc=<-2.1856 0.1725 0.5984>, rot=<0.006 0.0168 -0.0329 -0.9993>}
 middle_01_r {parent=middle_metacarpal_r, loc=<-7.3275 0.2447 -0.0129>, rot=<-0.0208 0.0042 0.0858 -0.9961>}
 middle_02_r {parent=middle_01_r, loc=<-4.0123 -0.1401 -0.1175>, rot=<-0.0218 0.0258 0.3512 -0.9357>}
 middle_03_r {parent=middle_02_r, loc=<-2.6077 -0.0534 -0.1373>, rot=<0.0082 0.0188 0.1708 -0.9851>}
 ring_metacarpal_r {parent=hand_r, loc=<-2.2074 0.3625 -1.3638>, rot=<-0.0284 0.0722 -0.0458 -0.9959>}
 ring_01_r {parent=ring_metacarpal_r, loc=<-7.0295 0.6197 0.1446>, rot=<-0.0489 -0.0252 0.1892 -0.9804>}
 ring_02_r {parent=ring_01_r, loc=<-3.6174 -0.4647 0.1027>, rot=<0 0 0.2806 -0.9598>}
 ring_03_r {parent=ring_02_r, loc=<-2.4152 -0.114 -0.0199>, rot=<0 0 0.1189 -0.9929>}
 pinky_metacarpal_r {parent=hand_r, loc=<-1.7616 0.9753 -3.3814>, rot=<-0.0397 0.0814 -0.069 -0.9935>}
 pinky_01_r {parent=pinky_metacarpal_r, loc=<-7.1709 0.4831 0.1094>, rot=<-0.0654 -0.0035 0.2242 -0.9723>}
 pinky_02_r {parent=pinky_01_r, loc=<-2.8274 -0.0893 0.1043>, rot=<0.0037 0.0011 0.2854 -0.9584>}
 pinky_03_r {parent=pinky_02_r, loc=<-2.2046 -0.0307 0.0078>, rot=<-0.0002 0.0098 0.1445 -0.9895>}
 lowerarm_twist_02_r {parent=lowerarm_r, loc=<-10.6707 0 0>, rot=<0 0 0 -1>}
 lowerarm_twist_01_r {parent=lowerarm_r, loc=<-18.316 0 0>, rot=<0 0 0 -1>}
 upperarm_twist_01_r {parent=upperarm_r, loc=<-7.2135 0.1134 0>, rot=<0 0 0 -1>}
 deform_upperarm_bk_base_r {parent=upperarm_twist_01_r, loc=<6.4053 4.3746 -0.314>, rot=<0 0 0 -1>}
 deform_upperarm_bk_r {parent=deform_upperarm_bk_base_r, loc=<0 0 0>, rot=<0 0 0 1>}
 deform_upperarm_fr_base_r {parent=upperarm_twist_01_r, loc=<5.6312 -5.092 -0.3354>, rot=<0 0 0 -1>}
 deform_upperarm_fr_r {parent=deform_upperarm_fr_base_r, loc=<0 0 0>, rot=<0 0 0 1>}
 deform_upperarm_up_base_r {parent=upperarm_twist_01_r, loc=<8.891 -0.4825 -3.9517>, rot=<0 0 0 -1>}
 deform_upperarm_up_r {parent=deform_upperarm_up_base_r, loc=<0 0 0>, rot=<0 0 0 1>}
 upperarm_twist_02_r {parent=upperarm_r, loc=<-19.7998 0.1134 0>, rot=<0 0 0 -1>}
 attach_backpack {parent=spine_05, loc=<-0.123 -13.473 0>, rot=<-0.0093 -0.707 0.0093 -0.707>}
 Backpack_BR {parent=attach_backpack, type=Null, loc=<0 0 0>, rot=<0 0 0 1>}
 attach_cape {parent=spine_05, loc=<9.017 -11.714 0>, rot=<0.0616 -0.7044 -0.0616 -0.7044>}
 cape {parent=attach_cape, type=Null, loc=<0 0 0>, rot=<0 0 0 1>}
 dyn_necklace_01 {parent=spine_05, loc=<7.4648 9.1167 0.0123>, rot=<0.4545 0.5417 0.5417 0.4545>}
 dyn_necklace_02 {parent=spine_05, loc=<7.8249 8.181 -3.8023>, rot=<-0.2624 -0.6889 -0.4301 -0.5212>}
 dyn_necklace_03 {parent=spine_05, loc=<8.8827 5.8185 -6.7228>, rot=<-0.153 -0.6903 -0.3265 -0.6272>}
 dyn_necklace_04 {parent=spine_05, loc=<10.1301 2.6067 -8.5663>, rot=<0.0562 -0.7341 -0.1254 -0.665>}
 dyn_necklace_05 {parent=spine_05, loc=<11.4828 -1.0052 -8.1423>, rot=<0.2032 -0.7366 0.0322 -0.6443>}
 dyn_necklace_06 {parent=spine_05, loc=<12.372 -3.9623 -5.5711>, rot=<0.363 -0.654 0.3404 -0.5698>}
 dyn_necklace_07 {parent=spine_05, loc=<12.8788 -5.0741 -1.9381>, rot=<0.4777 -0.5213 0.5213 -0.4777>}
 dyn_necklace_08 {parent=spine_05, loc=<12.8766 -5.0623 1.9379>, rot=<0.5175 -0.4739 0.5648 -0.4343>}
 dyn_necklace_09 {parent=spine_05, loc=<12.3562 -3.9774 5.5854>, rot=<0.5851 -0.3686 0.6385 -0.3378>}
 dyn_necklace_010 {parent=spine_05, loc=<11.473 -1.0164 8.1293>, rot=<0.6443 -0.0322 0.7366 -0.2032>}
 dyn_necklace_011 {parent=spine_05, loc=<10.1637 2.6166 8.5879>, rot=<0.6698 0.0962 0.731 -0.0882>}
 dyn_necklace_012 {parent=spine_05, loc=<8.888 5.7839 6.7084>, rot=<0.6395 0.2857 0.7097 0.0758>}
 dyn_necklace_013 {parent=spine_05, loc=<7.8313 8.1772 3.8209>, rot=<0.5573 0.4462 0.655 0.2476>}
 thigh_l {parent=pelvis, loc=<-1.0855 -0.2109 -8.7053>, rot=<0.0009 0.0355 0.0155 -0.9993>}
 calf_l {parent=thigh_l, loc=<-42.5284 -0.0019 -0.7297>, rot=<0.0006 -0.0161 -0.0391 -0.9991>}
 foot_l {parent=calf_l, loc=<-46.7401 -2.5014 -0.592>, rot=<0.0427 -0.0238 0.0353 -0.9982>}
 ball_l {parent=foot_l, loc=<-3.5658 10.8066 -0.1544>, rot=<0.0001 -0.0001 -0.7063 -0.7079>}
 calf_twist_02_l {parent=calf_l, loc=<-19.1546 -1.7943 -0.2638>, rot=<0.0237 0.0095 -0.0035 -0.9997>}
 calf_twist_01_l {parent=calf_l, loc=<-35.8483 -1.9197 -0.5794>, rot=<0.0237 0.0095 -0.0035 -0.9997>}
 thigh_twist_01_l {parent=thigh_l, loc=<-27.952 0 0>, rot=<0.0013 0 0 -1>}
 dyn_dress_fr_l_01 {parent=thigh_l, loc=<0.0364 7.5 0.0027>, rot=<-0.0254 -0.817 -0.5758 0.0171>}
 dyn_dress_fr_l_02 {parent=dyn_dress_fr_l_01, loc=<42.4121 -3.1848 0.4871>, rot=<0 -0.0342 0.003 -0.9994>}
 dyn_dress_fr_l_03 {parent=dyn_dress_fr_l_02, loc=<19.0039 -1.4662 -0.9516>, rot=<-0.0204 -0.0187 0.0255 -0.9993>}
 dyn_dress_bck_l_01 {parent=thigh_l, loc=<-0.057 -11.7582 -0.0042>, rot=<-0.0261 -0.8273 0.5579 0.0606>}
 dyn_dress_bck_l_02 {parent=dyn_dress_bck_l_01, loc=<42.2086 2.0809 4.8825>, rot=<0.0522 -0.9909 -0.1117 -0.054>}
 dyn_dress_bck_l_03 {parent=dyn_dress_bck_l_02, loc=<-18.8724 -1.5517 -0.4934>, rot=<0.0488 -0.997 -0.0579 -0.0138>}
 dyn_dress_side_l_01 {parent=thigh_l, loc=<-0.5161 0.0051 -7.2596>, rot=<-0.0172 -0.9985 0.0309 0.0429>}
 dyn_dress_side_l_02 {parent=dyn_dress_side_l_01, loc=<42.2966 -1.3595 3.5616>, rot=<0 -0.0069 0.004 -1>}
 dyn_dress_side_l_03 {parent=dyn_dress_side_l_02, loc=<18.8227 -3.8238 0.5696>, rot=<0.0655 0.0055 0.0835 -0.9943>}
 thigh_r {parent=pelvis, loc=<-1.0855 -0.2109 8.7053>, rot=<0.0355 -0.0009 0.9993 0.0155>}
 thigh_twist_01_r {parent=thigh_r, loc=<27.952 0 0>, rot=<0.0013 0 0 -1>}
 dyn_dress_fr_r_01 {parent=thigh_r, loc=<-0.0364 -7.5 -0.0027>, rot=<-0.5751 -0.0171 0.0254 -0.8175>}
 dyn_dress_fr_r_02 {parent=dyn_dress_fr_r_01, loc=<42.4123 3.1817 0.4914>, rot=<0 -0.0342 -0.003 -0.9994>}
 dyn_dress_fr_r_03 {parent=dyn_dress_fr_r_02, loc=<19.0038 1.4688 -0.9489>, rot=<0.0192 -0.0186 -0.0256 -0.9993>}
 dyn_dress_bck_r_01 {parent=thigh_r, loc=<0.057 11.7582 0.0042>, rot=<0.521 -0.0594 0.0287 -0.851>}
 dyn_dress_bck_r_02 {parent=dyn_dress_bck_r_01, loc=<42.2085 -1.6448 5.0469>, rot=<0.0436 0.0304 -0.0067 -0.9986>}
 dyn_dress_bck_r_03 {parent=dyn_dress_bck_r_02, loc=<18.8998 -0.3113 1.2328>, rot=<0.032 0.0371 -0.0112 -0.9987>}
 calf_r {parent=thigh_r, loc=<42.5284 0.0019 0.7297>, rot=<0.0006 -0.0161 -0.0391 -0.9991>}
 foot_r {parent=calf_r, loc=<46.7401 2.5014 0.592>, rot=<0.0427 -0.0238 0.0353 -0.9982>}
 ball_r {parent=foot_r, loc=<3.5658 -10.8066 0.1544>, rot=<0.0001 -0.0001 -0.7063 -0.7079>}
 calf_twist_02_r {parent=calf_r, loc=<19.1546 1.7943 0.2638>, rot=<0.0237 0.0095 -0.0035 -0.9997>}
 calf_twist_01_r {parent=calf_r, loc=<35.8483 1.9197 0.5794>, rot=<0.0237 0.0095 -0.0035 -0.9997>}
 dyn_dress_side_r_01 {parent=thigh_r, loc=<0.5069 -0.005 7.1294>, rot=<-0.0556 -0.0412 0.0209 -0.9974>}
 dyn_dress_side_r_02 {parent=dyn_dress_side_r_01, loc=<42.2951 1.9665 3.3942>, rot=<0 -0.0075 -0.0027 -1>}
 dyn_dress_side_r_03 {parent=dyn_dress_side_r_02, loc=<18.8215 3.8697 -0.054>, rot=<-0.0662 -0.009 -0.0832 -0.9943>}
 dyn_dress_fr_c_01 {parent=pelvis, loc=<0.2004 8.3644 0>, rot=<-0.0189 -0.7069 -0.7069 -0.0189>}
 dyn_dress_fr_c_02 {parent=dyn_dress_fr_c_01, loc=<43.5054 0 0.428>, rot=<0 -0.0327 0 -0.9995>}
 dyn_dress_fr_c_03 {parent=dyn_dress_fr_c_02, loc=<18.8364 0 -0.9014>, rot=<0 -0.0236 0 -0.9997>}
 dyn_dress_bck_c_01 {parent=pelvis, loc=<-0.6084 -11.8588 0.7801>, rot=<-0.0612 -0.7045 0.7045 0.0612>}
 dyn_dress_bck_c_02 {parent=dyn_dress_bck_c_01, loc=<42.6773 0 6.8052>, rot=<0 0.0423 0 -0.9991>}
 dyn_dress_bck_c_03 {parent=dyn_dress_bck_c_02, loc=<18.8029 0 1.7079>, rot=<0 0.0459 0 -0.9989>}
]
variables [
 ClavUp {type=double, default=0}
 Is Crouching {type=bool, default=False}
 Is On Grind Rail {type=bool, default=False}
 Is Playing Emote {type=double, default=0}
 L Crouch Strafe Behind {type=double, default=0}
 L Crouch Strafe Front {type=double, default=0}
 L calfBend Alpha {type=float, default=0}
 L thighBack Alpha {type=double, default=0}
 L thighForward Alpha {type=double, default=0}
 L thighSide Alpha {type=double, default=0}
 L thighSideIn_Bwd Alpha {type=double, default=0}
 L thighSideIn_Fwd Alpha {type=double, default=0}
 L thighTwistIn Alpha {type=double, default=0}
 L thighTwistOut Alpha {type=double, default=0}
 Low To Ground Alpha {type=float, default=0}
 Moving Alpha {type=double, default=0}
 Moving Up Aggressively Alpha {type=double, default=0}
 Not Skydive Alpha {type=double, default=1}
 R Crouch Strafe Behind {type=double, default=0}
 R Crouch Strafe Front {type=double, default=0}
 R calfBend Alpha {type=float, default=0}
 R thighBack Alpha {type=double, default=0}
 R thighForward Alpha {type=double, default=0}
 R thighSide Alpha {type=double, default=0}
 R thighSideIn_Bwd Alpha {type=double, default=0}
 R thighSideIn_Fwd Alpha {type=double, default=0}
 R thighTwistIn Alpha {type=double, default=0}
 Slope Info Input Slope Rotation {type=FRotator, default=(Pitch=0,Yaw=0,Roll=0)}
 Slope Pitch Sideways {type=double, default=0}
 Slope Roll Down {type=double, default=0}
 ThighSplit L Fwd {type=double, default=0}
 ThighSplit R Fwd {type=double, default=0}
 feet apart alpha {type=double, default=0}
]
