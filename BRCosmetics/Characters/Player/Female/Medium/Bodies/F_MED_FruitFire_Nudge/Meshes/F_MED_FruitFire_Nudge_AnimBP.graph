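# Post-process-style cosmetic Anim Blueprint for the F_MED_FruitFire_Nudge character body, copying the attached parent mesh's pose into its Control Rig.
# The rig receives grind, crouch, movement, skydive, emote and slope inputs, set from pawn state and events.
# Handle Slope Warp interpolates slope rotation and derives roll and pitch alphas; its smoothed root translation is not wired to the rig.
# A necklace bone constraint chain exists but is disconnected from the output pose.
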AnimBlueprint /BRCosmetics/Characters/Player/Female/Medium/Bodies/F_MED_FruitFire_Nudge/Meshes/F_MED_FruitFire_Nudge_AnimBP
target_skeleton=&/F_MED_FruitFire_Nudge_Skeleton
graphs [
 AnimGraph {nodes [
   n1 = "Constraint - dyn_necklace_09 (PPPP)"!(
    ComponentPose
    Alpha=1
    ConstraintWeights_0=0.6337
    ConstraintWeights_1:float=0.0008
    ConstraintWeights_2:float=0.3609
    ConstraintWeights_3:float=0.0045
   ) -> (Pose) {
    boneToModify {boneName=dyn_necklace_09}
    constraintSetup [
     {targetBone {boneName=spine_05}}
     {targetBone {boneName=clavicle_l}}
     {targetBone {boneName=clavicle_r}}
     {targetBone {boneName=upperarm_twist_01_r}}
    ]
    constraintWeights [1,1,1,1]
   }
   n2 = "Constraint - dyn_necklace_010 (PPPPP)"!(
    ComponentPose=@n1
    Alpha=0.5
    ConstraintWeights_0=0.8142
    ConstraintWeights_1:float=0.0037
    ConstraintWeights_2:float=0.1303
    ConstraintWeights_3:float=0.0402
    ConstraintWeights_4:float=0.0116
   ) -> (Pose) {
    boneToModify {boneName=dyn_necklace_010}
    constraintSetup [
     {targetBone {boneName=spine_05}}
     {targetBone {boneName=clavicle_r}}
     {targetBone {boneName=neck_01}}
     {targetBone {boneName=head}}
     {targetBone {boneName=neck_02}}
    ]
    constraintWeights [1,1,1,1,1]
   }
   n3 = "Constraint - dyn_necklace_010 (PPPP)"!(
    ComponentPose=@n2
    Alpha=0.5
    ConstraintWeights_0=0.4199
    ConstraintWeights_1:float=0.5463
    ConstraintWeights_2:float=0.0204
    ConstraintWeights_3:float=0.0135
   ) -> (Pose) {
    boneToModify {boneName=dyn_necklace_010}
    constraintSetup [
     {targetBone {boneName=spine_05}}
     {targetBone {boneName=clavicle_r}}
     {targetBone {boneName=upperarm_twist_01_r}}
     {targetBone {boneName=spine_04}}
    ]
    constraintWeights [1,1,1,1]
   }
   "Constraint - dyn_necklace_011 (PPPP)"!(
    ComponentPose=@n3
    Alpha=1
    ConstraintWeights_0=0.4432
    ConstraintWeights_1:float=0.4508
    ConstraintWeights_2:float=0.0245
    ConstraintWeights_3:float=0.0816
   ) -> (Pose) {
    boneToModify {boneName=dyn_necklace_011}
    constraintSetup [
     {targetBone {boneName=spine_05}}
     {targetBone {boneName=clavicle_r}}
     {targetBone {boneName=upperarm_twist_01_r}}
     {targetBone {boneName=spine_04}}
    ]
    constraintWeights [1,1,1,1]
   }
   n4 = Control Rig!(
    Source=Copy Pose From Mesh() {bUseAttachedParent=true}
    Alpha
    Is On Grind Rail:bool=Get bIsOnGrindRail()
    Is Crouching:bool=Get bIsCrouching()
    Moving Alpha:double=Get Moving Alpha() -> (Moving Alpha:double)
    Not Skydive Alpha:double=Get Not Skydive Alpha() -> (Not Skydive Alpha:double)
    Moving Up Aggressively Alpha:double=Get Moving Up Aggressively Alpha() -> (Moving Up Aggressively Alpha:double)
    Slope Info Input Slope Rotation:Rotator=Get Slope Info Input Slope Rotation() -> (Slope Info Input Slope Rotation:Rotator)
    Slope Roll Down:double=Get Slope Roll Down() -> (Slope Roll Down:double)
    Slope Pitch Sideways:double=Get Slope Pitch Sideways() -> (Slope Pitch Sideways:double)
    Is Playing Emote:double=Get Is Playing Emote() -> (Is Playing Emote:double)
   ) -> (Pose) {controlRigAssetReference {blueprintRigClass=&/F_MED_FruitFire_Nudge_CtrlRig.F_MED_FruitFire_Nudge_CtrlRig_C}}
   Output Pose(Result=Component To Local(ComponentPose=Local To Component(LocalPose=@n4))) {layerGroup=None}
  ]}
 EventGraph {nodes [
   n1 = Event BlueprintUpdateAnimation()
   > n2 = Sequence!(execute) -> (then_0, then_1, then_2:Exec)
   n3 = BreakVector(InVec=Get PawnVelocity())
   n4 = "Absolute (Float)"(A=@n3)
   n5 = "Absolute (Float)"(A=@n3.Y)
   n6 = "Absolute (Float)"(A=@n3.Z)
   > n7 = Branch(Condition=OR Boolean!(A="float > float"(A=@n4), B="float > float"(A=@n5), C:bool="float > float"(A=@n6)) -> (ReturnValue))
   > Set Moving Alpha(Moving Alpha:double=1) -> (Output_Get:double)
   Set Moving Alpha(execute=@n7.else, Moving Alpha=0)
   Set Moving Up Aggressively Alpha(execute=@n2.then_1, Moving Up Aggressively Alpha:double=MapRangeClamped(Value=@n3.Z, InRangeA=4000, InRangeB=8000)) -> (Output_Get:double)
   Handle Slope Warp!(execute=@n2.then_2, self, DeltaTime:double=@n1.DeltaTimeX) -> (then)
   Event OnBeginSkydiving()
   > Set Not Skydive Alpha(Not Skydive Alpha:double=0) -> (Output_Get:double)
   Event OnEndSkydiving()
   > Set Not Skydive Alpha(Not Skydive Alpha=1)
   Event OnBeginPlayingEmote() {is=title K2Node_Event}
   > Set Is Playing Emote(Is Playing Emote:double=1) -> (Output_Get:double)
   Event OnEndPlayingEmote() {is=title K2Node_Event}
   > Set Is Playing Emote(Is Playing Emote=0)
  ]}
 Handle Slope Warp {nodes [
   n1 = Handle Slope Warp!() -> (then, DeltaTime:double=~) {is=title K2Node_FunctionEntry}
   > n2 = Sequence!(execute) -> (then_0, then_1)
   n3 = Get bIsOnGround() -> (bIsOnGround:bool)
   > n4 = Branch(Condition=OR Boolean(A=NOT Boolean(A=@n3), B=Get bIsPlayingEmote(), C=Get bIsInVehicle() -> (bIsInVehicle:bool)))
   > Set Target Info Input Root Slope Translation(Target Info Input Root Slope Translation:Vector="0,0,-100") -> (Output_Get:Vector)
   n5 = Set Slope Info Input Slope Rotation(
    execute=@n4.else
    Slope Info Input Slope Rotation:Rotator=RInterpTo!(
     Current:Rotator=Get Slope Info Input Slope Rotation()
     Target:Rotator=Get!() -> (
      SlopeInfoInput_RootSlopeTranslation:Vector=~
      SlopeInfoInput_LeftFootSlopeTranslation:Vector=~
      SlopeInfoInput_RightFootSlopeTranslation:Vector=~
      SlopeInfoInput_SlopeRotation:Rotator=~
     ).SlopeInfoInput_SlopeRotation
     DeltaTime:float=@n1
     InterpSpeed:float=7
    ) -> (ReturnValue:Rotator)
   ) -> (Output_Get:Rotator)
   > Set Target Info Input Root Slope Translation(Target Info Input Root Slope Translation="0,0,5")
   > Set Slope Roll Down(Slope Roll Down:double=MapRangeClamped(Value=BreakRotator(InRot=@n5), InRangeA=5, InRangeB=30)) -> (Output_Get:double)
   n6 = BreakRotator(InRot=@n5)
   > Set Slope Pitch Sideways(Slope Pitch Sideways:double=MapRangeClamped(Value="Absolute (Float)"(A=@n6.Pitch), InRangeA=8, InRangeB=30)) -> (Output_Get:double)
   Set Slope Info Input Root Slope Translation(
    execute=@n2.then_1
    Slope Info Input Root Slope Translation:Vector=VInterpTo!(
     Current:Vector=Get Slope Info Input Root Slope Translation() -> (Slope Info Input Root Slope Translation:Vector)
     Target:Vector=Get Target Info Input Root Slope Translation() -> (Target Info Input Root Slope Translation:Vector)
     DeltaTime:float=@n1
     InterpSpeed:float=10
    ) -> (ReturnValue:Vector)
   ) -> (Output_Get:Vector)
  ]}
]
variables [
 Delay Slope Calc {type=boolean, default=true}
 Is Playing Emote {type=number}
 Moving Alpha {type=number}
 Moving Up Aggressively Alpha {type=number}
 Not Skydive Alpha {type=number, default=1}
 sliding alpha {type=number}
 Slope Info Input Root Slope Translation {type=Vector}
 Slope Info Input Slope Rotation {type=Rotator}
 Slope Pitch Sideways {type=number}
 Slope Roll Down {type=number}
 Target Info Input Root Slope Translation {type=Vector}
]
