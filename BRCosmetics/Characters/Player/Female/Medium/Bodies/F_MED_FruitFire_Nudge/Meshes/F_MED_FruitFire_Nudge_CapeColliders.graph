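# Physics asset of cape colliders for the F_MED_FruitFire_Nudge female character body.
# Capsules cover the pelvis, spine, clavicles, head and feet, with paired capsules on spine_02 through spine_04.
# Large capsules also sit on the left, center and right three-bone back dress chains (dyn_dress_bck).
# Arm, thigh, calf and neck bodies hold no shapes, and no constraints join the bodies.
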PhysicsAsset /BRCosmetics/Characters/Player/Female/Medium/Bodies/F_MED_FruitFire_Nudge/Meshes/F_MED_FruitFire_Nudge_CapeColliders
bodies {
 calf_l []
 calf_r []
 clavicle_l [{shapeType=Capsule, length=6.5623, radius=7.5194, center=<3.9566 -1.0862 -2.1127>, rotation=<rot -77.5566 177.3915 7.7314>}]
 clavicle_r [{shapeType=Capsule, length=6.5623, radius=7.5194, center=<-3.9566 1.0862 2.1127>, rotation=<rot -77.5566 177.3915 -172.2686>}]
 dyn_dress_bck_c_01 [{shapeType=Capsule, length=5.5, radius=18.5, shapeName=dyn_dress_bck_c_01_capsule, center=<21.0369 0 -8.0176>, rotation=<rot -90 0 0>}]
 dyn_dress_bck_c_02 [{shapeType=Capsule, length=11.5, radius=14.5, shapeName=dyn_dress_bck_c_02_capsule, center=<0 0 -7.7528>, rotation=<rot 90 0 0>}]
 dyn_dress_bck_c_03 [{shapeType=Capsule, length=10.5, radius=16, shapeName=dyn_dress_bck_c_03_capsule, center=<1.1634 0 -12.1141>, rotation=<rot 90 0 0>}]
 dyn_dress_bck_l_01 [{shapeType=Capsule, length=12, radius=14, shapeName=dyn_dress_bck_l_01_capsule, center=<22.0419 0 -5.5805>, rotation=<rot -90 0.0001 0>}]
 dyn_dress_bck_l_02 [{shapeType=Capsule, length=10.5, radius=14.5, shapeName=dyn_dress_bck_l_02_capsule, center=<0 0 7.39>, rotation=<rot -90 0 0>}]
 dyn_dress_bck_l_03 [{shapeType=Capsule, length=10.5, radius=14.5, shapeName=dyn_dress_bck_l_03_capsule, center=<0 0 -6.9091>, rotation=<rot -90 0.0001 0>}]
 dyn_dress_bck_r_01 [{
   shapeType=Capsule
   length=12
   radius=14
   shapeName=dyn_dress_bck_r_01_capsule
   center=<22.0419 -0.0001 -4.7705>
   rotation=<rot -90 -0.0001 0>
  }]
 dyn_dress_bck_r_02 [{shapeType=Capsule, length=10.5, radius=14.5, shapeName=dyn_dress_bck_r_02_capsule, center=<0 0 -7.4936>, rotation=<rot -90 -0.0001 0>}]
 dyn_dress_bck_r_03 [{shapeType=Capsule, length=10.5, radius=14.5, shapeName=dyn_dress_bck_r_03_capsule, center=<0 0 -10.8798>, rotation=<rot -90 -0.0001 0>}]
 foot_l [{shapeType=Capsule, length=18.5765, radius=17.5344, center=<-4.1811 6.0027 -0.338>, rotation=<rot 0.0388 1.4701 -89.3982>}]
 foot_r [{shapeType=Capsule, length=15.675, radius=17.5344, center=<4.1811 -6.0027 0.338>, rotation=<rot 0.0388 1.4701 90.6018>}]
 head [{shapeType=Capsule, length=0, radius=18, center=<9.0698 0.0746 0>, rotation=<rot 0 -90 0>}]
 lowerarm_l []
 lowerarm_r []
 neck_02 []
 pelvis [{shapeType=Capsule, length=6.6042, radius=17.6441, center=<-2.4471 -2.5289 -0.1981>}]
 spine_01 [{shapeType=Capsule, length=4.3888, radius=14.3638, center=<-8.9021 -0.8973 0>}]
 spine_02 [
  {shapeType=Capsule, length=9.0219, radius=11.2317, center=<-2.6885 -1.2641 0>}
  {shapeType=Capsule, length=20.6875, radius=9.3125, center=<-1.8317 1.9362 0>, rotation=<rot 90 0 0>}
 ]
 spine_03 [
  {shapeType=Capsule, length=8.6863, radius=10.4854, center=<0.0574 0.4416 0>}
  {shapeType=Capsule, length=20.7188, radius=9.2812, center=<-1.2163 0.549 0>, rotation=<rot 90 180 -175.0732>}
 ]
 spine_04 [
  {shapeType=Capsule, length=9.0076, radius=10.3916, center=<5.6234 -0.6779 0>}
  {shapeType=Capsule, length=10.0312, radius=9.2812, center=<-1.2326 1.308 0>, rotation=<rot 90 0 0>}
 ]
 spine_05 [{shapeType=Capsule, length=8.178, radius=13.8571, center=<0.6969 1.91 0>}]
 thigh_l []
 thigh_r []
 upperarm_l []
 upperarm_r []
}
constraints []
