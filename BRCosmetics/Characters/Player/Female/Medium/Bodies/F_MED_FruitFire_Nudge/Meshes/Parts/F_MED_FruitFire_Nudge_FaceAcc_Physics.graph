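# Physics asset for the face accessory mesh of the FruitFire Nudge medium female character cosmetic, covering the head and a dynamic hair chain.
# Capsule bodies sit on the head and three chained hair bones, with the hair capsules all sharing the same size.
# Constraints link each hair bone to its parent, keeping the root segment tight to the head while the lower segments swing widely, with twist nearly locked throughout.
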PhysicsAsset /BRCosmetics/Characters/Player/Female/Medium/Bodies/F_MED_FruitFire_Nudge/Meshes/Parts/F_MED_FruitFire_Nudge_FaceAcc_Physics
bodies {
 dyn_hair_01 [{shapeType=Capsule, length=4, radius=14, center=<6.2863 0 0>, rotation=<rot 90 0 0>}]
 dyn_hair_02 [{shapeType=Capsule, length=4, radius=14, center=<6.2863 0 0>, rotation=<rot 90 0 0>}]
 dyn_hair_03 [{shapeType=Capsule, length=4, radius=14, center=<6.2863 0 0>, rotation=<rot 90 0 0>}]
 head [{shapeType=Capsule, length=3, radius=12.5, center=<4.0709 0 0>, rotation=<rot -90 -0.0002 0>}]
}
constraints [
 {bone1Name=dyn_hair_01, bone2Name=head, swing1LimitDegrees=10, swing2LimitDegrees=10, twistLimitDegrees=1}
 {bone1Name=dyn_hair_02, bone2Name=dyn_hair_01, swing1LimitDegrees=80, swing2LimitDegrees=60, twistLimitDegrees=1}
 {bone1Name=dyn_hair_03, bone2Name=dyn_hair_02, swing1LimitDegrees=80, swing2LimitDegrees=60, twistLimitDegrees=1}
]
