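# Animation Blueprint for the face accessory mesh part of the F_MED_FruitFire_Nudge character, adding secondary physics motion.
# It copies the pose from the attached parent mesh, then runs a RigidBody simulation in neck_01 bone space with low world-motion influence.
# The event graph stores a parameter-asset-driven gravity override each update and changes Normal Alpha on sprint, jump and fall events.
# Neither variable is visibly wired into the RigidBody node, which shows fixed gravity.
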AnimBlueprint /BRCosmetics/Characters/Player/Female/Medium/Bodies/F_MED_FruitFire_Nudge/Meshes/Parts/F_MED_FruitFire_Nudge_FaceAcc_AnimBP
target_skeleton=&/F_MED_FruitFire_Nudge_FaceAcc_Skeleton
graphs [
 AnimGraph {nodes [
   n1 = RigidBody!(
    ComponentPose=Local To Component(LocalPose=Copy Pose From Mesh() {bUseAttachedParent=true})
    Alpha
    OverrideWorldGravity:Vector="0,0,-980"
    ExternalForce
    SimSpaceSettings:Sim Space Settings Structure=(WorldAlpha=0.05,VelocityScaleZ=1,DampingAlpha=1,MaxLinearVelocity=10000,MaxAngularVelocity=10000,MaxLinearAcceleration=10000,MaxAngularAcceleration=10000,ExternalLinearDragV=(X=0,Y=0,Z=0),ExternalLinearVelocity=(X=0,Y=0,Z=0),ExternalAngularVelocity=(X=0,Y=0,Z=0))
   ) -> (Pose) {
    bClampLinearTranslationLimitToRefPose=true
    bOverrideWorldGravity=true
    overridePhysicsAsset=&/F_MED_FruitFire_Nudge_FaceAcc_Physics
    simulationSpace=BaseBoneSpace
    alphaScaleBiasClamp {bInterpResult=true}
    baseBoneRef {boneName=neck_01}
    componentAppliedLinearAccClamp=<1000 1000 1000>
    overrideWorldGravity {z=-980}
    simSpaceSettings {worldAlpha=0.05}
   }
   Output Pose(Result=Component To Local(ComponentPose=@n1)) {layerGroup=None}
  ]}
 EventGraph {nodes [
   n1 = Event BlueprintUpdateAnimation()
   > n2 = "Gravity Override State Function (Parameter Asset) (Nativized)"(
    CharacterPartAnimInstance=Self-Reference()
    GameTimeInSeconds=@n1.DeltaTimeX
    Parameters=&/F_Med_FruitFire_Nudge_FaceAcc_Parameters
   )
   > Set Gravity Override(Gravity Override=@n2)
   Event OnBeginTacticalSprinting() {is=title K2Node_Event}
   > n3 = Branch()
   > Set Normal Alpha(Normal Alpha:double=0.5) -> (Output_Get:double)
   Set Normal Alpha(execute=@n3.else, Normal Alpha=1)
   n4 = Event OnEndTacticalSprinting() {is=title K2Node_Event}
   Set Normal Alpha(execute=[@n4.then, @n8.else], Normal Alpha=1)
   n5 = Event OnBeginJumping()
   n6 = Branch(execute=[@n5.then, @n9.then], Condition=Get bIsTacticalSprinting())
   > Set Normal Alpha(Normal Alpha=0.3)
   Set Normal Alpha(execute=@n6.else, Normal Alpha=1)
   n7 = Event OnEndJumping()
   n8 = Branch(execute=[@n7.then, @n10.then], Condition=Get bIsTacticalSprinting())
   > Set Normal Alpha(Normal Alpha=0.3)
   n9 = Event OnBeginFalling()
   n10 = Event OnEndFalling()
  ]}
]
variables [
 Gravity Override {type=Vector}
 Normal Alpha {type=number, default=1}
 Sim Space Settings {
  type=SimSpaceSettings
  default {
   dampingAlpha=0
   maxAngularAcceleration=50
   maxAngularVelocity=5
   maxLinearAcceleration=50
   maxLinearVelocity=80
   velocityScaleZ=1
   worldAlpha=0.05
   externalLinearDragV=<0 0 0>
  }
 }
 Sprinting Alpha {type=number}
]
